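# Physics asset for the F_MED_KnightCatRacket female character body, setting up collision and joint limits for its dynamic necklace and skirt bones.
# Capsules on torso, clavicles, neck, thighs and feet, plus a large root sphere, accompany the dynamic chains.
# A three-link necklace hangs from spine_05; two-segment skirt chains attach through small MSTR bodies.
# Back, side and center-front panels anchor to pelvis, other front panels to thighs, with one free-swing constraint joining two right-front panels.
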PhysicsAsset /BRCosmetics/Characters/Player/Female/Medium/Bodies/F_MED_KnightCatRacket/Meshes/F_MED_KnightCatRacket_Physics
bodies {
 clavicle_l [{shapeType=Capsule, length=8.25, radius=4.5, shapeName=clavicle_l_capsule, center=<6.0254 -1 0.022>, rotation=<rot 82.8019 -0.0064 -0.0065>}]
 clavicle_r [{shapeType=Capsule, length=7.75, radius=5, shapeName=clavicle_r_capsule, center=<-5.8922 1 0.4801>, rotation=<rot 84.2666 -0.0041 -0.0042>}]
 dyn_necklace_01 [{shapeType=Capsule, length=9.88, radius=0.87, shapeName=dyn_necklace_01_capsule, center=<0 0.7698 0>}]
 dyn_necklace_01_MSTR [{shapeType=Capsule, length=1, radius=1, shapeName=dyn_necklace_01_MSTR_capsule}]
 dyn_necklace_02 [{shapeType=Capsule, length=6.5135, radius=0.7365, shapeName=dyn_necklace_02_capsule, center=<0 0.4848 0>}]
 dyn_necklace_03 [{
   shapeType=Capsule
   length=4.6709
   radius=0.5791
   shapeName=dyn_necklace_03_capsule
   center=<0.8889 0.2077 0>
   rotation=<rot 90 -165.9638 -155.9639>
  }]
 dyn_skirt_01_l [{shapeType=Capsule, length=7, radius=2, shapeName=dyn_skirt_01_l_capsule, center=<5.9929 0 2.4481>, rotation=<rot 90 0 0>}]
 dyn_skirt_01_l_MSTR [{shapeType=Capsule, length=1, radius=1, shapeName=dyn_skirt_01_l_MSTR_capsule}]
 dyn_skirt_01_r [{shapeType=Capsule, length=5.75, radius=3.25, shapeName=dyn_skirt_01_r_capsule, center=<5.9929 0 2.8412>, rotation=<rot 90 0 0>}]
 dyn_skirt_01_r_MSTR [{shapeType=Capsule, length=1, radius=1, shapeName=dyn_skirt_01_r_MSTR_capsule}]
 dyn_skirt_02_l [{shapeType=Capsule, length=6.75, radius=3.5, shapeName=dyn_skirt_02_l_capsule, center=<6.3896 0 0.8838>, rotation=<rot 90 0 0>}]
 dyn_skirt_02_r [{shapeType=Capsule, length=6.75, radius=3.5, shapeName=dyn_skirt_02_r_capsule, center=<6.3896 0 0.7339>, rotation=<rot 90 0 0>}]
 dyn_skirt_bk_01 [{shapeType=Capsule, length=6.5, radius=2.5, shapeName=dyn_skirt_bk_01_capsule, center=<5.9929 0 2.0401>, rotation=<rot 90 0 0>}]
 dyn_skirt_bk_01_MSTR [{shapeType=Capsule, length=1, radius=1, shapeName=dyn_skirt_bk_01_MSTR_capsule}]
 dyn_skirt_bk_01_l [{shapeType=Capsule, length=7, radius=2, shapeName=dyn_skirt_bk_01_l_capsule, center=<5.9929 0 2.4481>, rotation=<rot 90 0 0>}]
 dyn_skirt_bk_01_l_MSTR [{shapeType=Capsule, length=1, radius=1, shapeName=dyn_skirt_bk_01_l_MSTR_capsule}]
 dyn_skirt_bk_01_r [{shapeType=Capsule, length=7, radius=2, shapeName=dyn_skirt_bk_01_r_capsule, center=<5.9929 0 2.0401>, rotation=<rot 90 0 0>}]
 dyn_skirt_bk_01_r_MSTR [{shapeType=Capsule, length=1, radius=1, shapeName=dyn_skirt_bk_01_r_MSTR_capsule}]
 dyn_skirt_bk_02 [{shapeType=Capsule, length=5.5, radius=4.75, shapeName=dyn_skirt_bk_02_capsule, center=<6.3896 0 -0.1196>, rotation=<rot 90 0 0>}]
 dyn_skirt_bk_02_l [{shapeType=Capsule, length=6.75, radius=3.5, shapeName=dyn_skirt_bk_02_l_capsule, center=<6.3896 0 0.7537>, rotation=<rot 90 0 0>}]
 dyn_skirt_bk_02_r [{shapeType=Capsule, length=6.75, radius=3.5, shapeName=dyn_skirt_bk_02_r_capsule, center=<6.3896 0 1.6691>, rotation=<rot 90 0 0>}]
 dyn_skirt_fr2_01_l [{shapeType=Capsule, length=7.25, radius=3.75, shapeName=dyn_skirt_fr2_01_l_capsule, center=<6.0575 0 5.1519>, rotation=<rot 90 0 0>}]
 dyn_skirt_fr2_01_l_MSTR [{shapeType=Capsule, length=1, radius=1, shapeName=dyn_skirt_fr2_01_l_MSTR_capsule}]
 dyn_skirt_fr2_01_r [{shapeType=Capsule, length=7.25, radius=3.75, shapeName=dyn_skirt_fr2_01_r_capsule, center=<6.0575 0 2.7872>, rotation=<rot 90 0 0>}]
 dyn_skirt_fr2_01_r_MSTR [{shapeType=Capsule, length=1, radius=1, shapeName=dyn_skirt_fr2_01_r_MSTR_capsule}]
 dyn_skirt_fr2_02_l [{
   shapeType=Capsule
   length=5.25
   radius=5.75
   shapeName=dyn_skirt_fr2_02_l_capsule
   center=<6.0575 0 5.1339>
   rotation=<rot 90 -158.7859 -68.786>
  }]
 dyn_skirt_fr2_02_r [{shapeType=Capsule, length=6, radius=4.75, shapeName=dyn_skirt_fr2_02_r_capsule, center=<4.9856 0 2.618>, rotation=<rot 90 0 0>}]
 dyn_skirt_fr_01 [{shapeType=Capsule, length=3.75, radius=3.75, shapeName=dyn_skirt_fr_01_capsule, center=<8.489 0 5.8238>, rotation=<rot 90 180 90>}]
 dyn_skirt_fr_01_MSTR [{shapeType=Capsule, length=1, radius=1, shapeName=dyn_skirt_fr_01_MSTR_capsule}]
 dyn_skirt_fr_01_l [{shapeType=Capsule, length=4.25, radius=3.25, shapeName=dyn_skirt_fr_01_l_capsule, center=<6.0575 0 5.0008>, rotation=<rot 90 0 0>}]
 dyn_skirt_fr_01_l_MSTR [{shapeType=Capsule, length=1, radius=1, shapeName=dyn_skirt_fr_01_l_MSTR_capsule}]
 dyn_skirt_fr_01_r [{shapeType=Capsule, length=7.75, radius=4.25, shapeName=dyn_skirt_fr_01_r_capsule, center=<7.6421 0 2.2075>, rotation=<rot 90 0 0>}]
 dyn_skirt_fr_01_r_MSTR [{shapeType=Capsule, length=1, radius=1, shapeName=dyn_skirt_fr_01_r_MSTR_capsule}]
 dyn_skirt_fr_02 [{shapeType=Capsule, length=4.5, radius=7.25, shapeName=dyn_skirt_fr_02_capsule, center=<2.5813 0 1.8395>, rotation=<rot 90 180 -180>}]
 dyn_skirt_fr_02_l [{shapeType=Capsule, length=7.5, radius=6.5, shapeName=dyn_skirt_fr_02_l_capsule, center=<2.1611 0 3.978>, rotation=<rot 90 180 -0.0001>}]
 dyn_skirt_fr_02_r [{
   shapeType=Capsule
   length=7.5
   radius=4.75
   shapeName=dyn_skirt_fr_02_r_capsule
   center=<5.3175 0.1914 -1.6601>
   rotation=<rot 90 0 -179.9999>
  }]
 foot_l [{shapeType=Capsule, length=13, radius=14.5, shapeName=foot_l_capsule, center=<-0.7444 4.5327 0.5135>, rotation=<rot 0 -179.223 83.3786>}]
 foot_r [{shapeType=Capsule, length=13, radius=14.5, shapeName=foot_r_capsule, center=<0.3475 -5.4849 -0.6368>, rotation=<rot 0.4048 3.4895 -96.6096>}]
 neck_01 [{shapeType=Capsule, length=10.25, radius=4.75, shapeName=neck_01_capsule, center=<0 -2.3915 0>, rotation=<rot 90 0 0>}]
 pelvis [{shapeType=Capsule, length=11.4032, radius=8}, {shapeType=Capsule, length=8.5, radius=10.75, shapeName=pelvis_capsule, center=<-2.0799 0 0>}]
 root [{shapeType=Sphere, radius=90.25, shapeName=root_sphere, center=<0 9.2013 -71.5106>}]
 spine_04 [{shapeType=Capsule, length=6.1229, radius=9.6271, shapeName=spine_04_capsule, center=<3.9618 0.2402 0>}]
 spine_05 [{shapeType=Capsule, length=15.6423, radius=8.1893, center=<1.4872 -2.1508 0.0457>, rotation=<rot 0.9448 -90 0.0715>}]
 thigh_l [{shapeType=Capsule, length=26.7722, radius=6.7328, center=<-20.1656 0 0>, rotation=<rot 90 0 0>}]
 thigh_r [{shapeType=Capsule, length=26.2407, radius=6.9552, center=<20.5804 0 0>, rotation=<rot 90 0 0>}]
}
constraints [
 {bone1Name=thigh_l, bone2Name=pelvis, swing1LimitDegrees=45, swing2LimitDegrees=45, twistLimitDegrees=45}
 {bone1Name=thigh_r, bone2Name=pelvis, swing1LimitDegrees=45, swing2LimitDegrees=45, twistLimitDegrees=45}
 {bone1Name=dyn_necklace_02, bone2Name=dyn_necklace_01, swing1LimitDegrees=25, swing2LimitDegrees=35, twistLimitDegrees=15}
 {bone1Name=dyn_necklace_03, bone2Name=dyn_necklace_02, swing1LimitDegrees=60, swing2LimitDegrees=60, twistLimitDegrees=60}
 {bone1Name=dyn_skirt_bk_02, bone2Name=dyn_skirt_bk_01, swing1LimitDegrees=35, swing2LimitDegrees=120, twistLimitDegrees=5}
 {bone1Name=dyn_skirt_bk_02_l, bone2Name=dyn_skirt_bk_01_l, swing1LimitDegrees=35, swing2LimitDegrees=120, twistLimitDegrees=5}
 {bone1Name=dyn_skirt_bk_02_r, bone2Name=dyn_skirt_bk_01_r, swing1LimitDegrees=35, swing2LimitDegrees=120, twistLimitDegrees=5}
 {bone1Name=dyn_skirt_fr_02, bone2Name=dyn_skirt_fr_01, swing1LimitDegrees=15, swing2LimitDegrees=45, twistLimitDegrees=5}
 {bone1Name=dyn_skirt_fr_02_l, bone2Name=dyn_skirt_fr_01_l, swing1LimitDegrees=15, swing2LimitDegrees=90, twistLimitDegrees=5}
 {bone1Name=dyn_skirt_fr_02_r, bone2Name=dyn_skirt_fr_01_r, swing1LimitDegrees=15, swing2LimitDegrees=90, twistLimitDegrees=5}
 {bone1Name=dyn_skirt_02_l, bone2Name=dyn_skirt_01_l, swing1LimitDegrees=35, swing2LimitDegrees=120, twistLimitDegrees=5}
 {bone1Name=dyn_skirt_02_r, bone2Name=dyn_skirt_01_r, swing1LimitDegrees=35, swing2LimitDegrees=120, twistLimitDegrees=5}
 {bone1Name=dyn_skirt_fr2_02_l, bone2Name=dyn_skirt_fr2_01_l, swing1LimitDegrees=15, swing2LimitDegrees=90, twistLimitDegrees=5}
 {bone1Name=dyn_skirt_fr2_02_r, bone2Name=dyn_skirt_fr2_01_r, swing1LimitDegrees=15, swing2LimitDegrees=90, twistLimitDegrees=5}
 {bone1Name=dyn_necklace_01, bone2Name=spine_05, swing1LimitDegrees=30.84, swing2LimitDegrees=13.64, twistLimitDegrees=5}
 {
  bone1Name=dyn_skirt_fr2_02_r
  bone2Name=dyn_skirt_fr_02_r
  swing1LimitDegrees=45
  swing1Motion=Free
  swing2LimitDegrees=45
  swing2Motion=Free
  twistLimitDegrees=45
  twistMotion=Free
 }
 {bone1Name=dyn_skirt_01_l, bone2Name=dyn_skirt_01_l_MSTR, swing1LimitDegrees=35, swing2LimitDegrees=120, twistLimitDegrees=5}
 {bone1Name=dyn_skirt_01_l_MSTR, bone2Name=pelvis, swing1LimitDegrees=35, swing2LimitDegrees=100, twistLimitDegrees=5}
 {bone1Name=dyn_skirt_01_r, bone2Name=dyn_skirt_01_r_MSTR, swing1LimitDegrees=35, swing2LimitDegrees=120, twistLimitDegrees=5}
 {bone1Name=dyn_skirt_01_r_MSTR, bone2Name=pelvis, swing1LimitDegrees=35, swing2LimitDegrees=100, twistLimitDegrees=5}
 {bone1Name=dyn_skirt_bk_01, bone2Name=dyn_skirt_bk_01_MSTR, swing1LimitDegrees=35, swing2LimitDegrees=120, twistLimitDegrees=5}
 {bone1Name=dyn_skirt_bk_01_MSTR, bone2Name=pelvis, swing1LimitDegrees=35, swing2LimitDegrees=100, twistLimitDegrees=5}
 {bone1Name=dyn_skirt_bk_01_l, bone2Name=dyn_skirt_bk_01_l_MSTR, swing1LimitDegrees=35, swing2LimitDegrees=120, twistLimitDegrees=5}
 {bone1Name=dyn_skirt_bk_01_l_MSTR, bone2Name=pelvis, swing1LimitDegrees=35, swing2LimitDegrees=100, twistLimitDegrees=5}
 {bone1Name=dyn_skirt_bk_01_r, bone2Name=dyn_skirt_bk_01_r_MSTR, swing1LimitDegrees=35, swing2LimitDegrees=120, twistLimitDegrees=5}
 {bone1Name=dyn_skirt_bk_01_r_MSTR, bone2Name=pelvis, swing1LimitDegrees=35, swing2LimitDegrees=100, twistLimitDegrees=5}
 {bone1Name=dyn_skirt_fr_01, bone2Name=dyn_skirt_fr_01_MSTR, swing1LimitDegrees=15, swing2LimitDegrees=45, twistLimitDegrees=5}
 {bone1Name=dyn_skirt_fr_01_MSTR, bone2Name=pelvis, swing1LimitDegrees=15, swing2LimitDegrees=45, twistLimitDegrees=5}
 {bone1Name=dyn_skirt_fr_01_l_MSTR, bone2Name=thigh_l, swing1LimitDegrees=15, swing2LimitDegrees=45, twistLimitDegrees=45}
 {bone1Name=dyn_skirt_fr2_01_l, bone2Name=dyn_skirt_fr2_01_l_MSTR, swing1LimitDegrees=15, swing2LimitDegrees=45, twistLimitDegrees=5}
 {bone1Name=dyn_skirt_fr_01_l, bone2Name=dyn_skirt_fr_01_l_MSTR, swing1LimitDegrees=15, swing2LimitDegrees=45, twistLimitDegrees=5}
 {bone1Name=dyn_skirt_fr2_01_l_MSTR, bone2Name=thigh_l, swing1LimitDegrees=15, swing2LimitDegrees=45, twistLimitDegrees=5}
 {bone1Name=dyn_skirt_fr2_01_r, bone2Name=dyn_skirt_fr2_01_r_MSTR, swing1LimitDegrees=15, swing2LimitDegrees=65, twistLimitDegrees=5}
 {bone1Name=dyn_skirt_fr2_01_r_MSTR, bone2Name=thigh_r, swing1LimitDegrees=15, swing2LimitDegrees=45, twistLimitDegrees=5}
 {bone1Name=dyn_skirt_fr_01_r, bone2Name=dyn_skirt_fr_01_r_MSTR, swing1LimitDegrees=15, swing2LimitDegrees=45, twistLimitDegrees=5}
 {bone1Name=dyn_skirt_fr_01_r_MSTR, bone2Name=thigh_r, swing1LimitDegrees=15, swing2LimitDegrees=45, twistLimitDegrees=5}
]
