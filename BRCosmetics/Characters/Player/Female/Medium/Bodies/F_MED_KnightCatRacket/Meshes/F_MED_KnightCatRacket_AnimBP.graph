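# Animation Blueprint for the F_MED_KnightCatRacket player body, adding secondary motion to its skirt bones.
# SkyDive_mode switches between a half-strength AnimDynamics chain with thigh sphere limits and a RigidBody simulation, both fed by a cached pose.
# The RigidBody branch layers thigh-following constraints and additive bone corrections, weighted by crouch, movement and rotation-history variables, against clipping.
# Another chain combines gravity-driven AnimDynamics with fixed-weight skirt constraints.
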AnimBlueprint /BRCosmetics/Characters/Player/Female/Medium/Bodies/F_MED_KnightCatRacket/Meshes/F_MED_KnightCatRacket_AnimBP
target_skeleton=&/F_MED_KnightCatRacket_Skeleton
graphs [
 AnimGraph {nodes [
   n1 = Anim Dynamics - Bone: dyn_skirt_bk_01_l(ComponentPose=Local To Component(LocalPose=Use cached pose 'SavedPose13862'()), Alpha=0.5, ExternalForce="0,0,0") {
    angularDampingOverride=0.9999
    angularSpringConstant=75
    bAngularSpring=true
    bGravityOverrideInSimSpace=true
    bOverrideAngularDamping=true
    bUseSphericalLimits=true
    simulationSpace=RootRelative
    alphaBoolBlend {blendOption=Linear}
    boundBone {boneName=dyn_skirt_bk_01_l}
    componentLinearAccScale=<1.2 1.2 1.2>
    componentLinearVelScale=<1.2 1.2 1.2>
    physicsBodyDefinitions [{
      sphereCollisionRadius=0
      boundBone {boneName=dyn_skirt_bk_01_l}
      boxExtents=<20 20 10>
      constraintSetup {angularLimitsMax=<25 0 25>, angularLimitsMin=<-25 -25 -25>, angularTarget=<0 0 0>}
      localJointOffset=<2 0 0>
     }]
    sphericalLimits [{limitRadius=20, drivingBone {boneName=thigh_l}, sphereLocalOffset=<-24.3408 0 0>}]
   }
   n2 = Anim Dynamics - Bone: dyn_skirt_bk_02_l(ComponentPose=@n1, Alpha=0.5, ExternalForce="0,0,0") {
    angularDampingOverride=0.9999
    angularSpringConstant=75
    bAngularSpring=true
    bGravityOverrideInSimSpace=true
    bOverrideAngularDamping=true
    bUseSphericalLimits=true
    simulationSpace=RootRelative
    alphaBoolBlend {blendOption=Linear}
    boundBone {boneName=dyn_skirt_bk_02_l}
    componentLinearAccScale=<1.2 1.2 1.2>
    componentLinearVelScale=<1.2 1.2 1.2>
    physicsBodyDefinitions [{
      sphereCollisionRadius=0
      boundBone {boneName=dyn_skirt_bk_02_l}
      boxExtents=<20 20 10>
      constraintSetup {angularLimitsMax=<25 0 25>, angularLimitsMin=<-25 -25 -25>, angularTarget=<0 0 0>}
      localJointOffset=<2 0 0>
     }]
    sphericalLimits [{limitRadius=20, drivingBone {boneName=thigh_l}, sphereLocalOffset=<-19.667 -0.6747 0>}]
   }
   n3 = Anim Dynamics - Bone: dyn_skirt_bk_01(ComponentPose=@n2, Alpha=0.5, ExternalForce="0,0,0") {
    angularDampingOverride=0.9999
    angularSpringConstant=75
    bAngularSpring=true
    bGravityOverrideInSimSpace=true
    bOverrideAngularDamping=true
    simulationSpace=RootRelative
    alphaBoolBlend {blendOption=Linear}
    boundBone {boneName=dyn_skirt_bk_01}
    componentLinearAccScale=<1.2 1.2 1.2>
    componentLinearVelScale=<1.2 1.2 1.2>
    physicsBodyDefinitions [{
      sphereCollisionRadius=0
      boundBone {boneName=dyn_skirt_bk_01}
      boxExtents=<20 20 10>
      constraintSetup {angularLimitsMax=<25 0 25>, angularLimitsMin=<-25 -25 -25>, angularTarget=<0 0 0>}
      localJointOffset=<2 0 0>
     }]
   }
   n4 = Anim Dynamics - Bone: dyn_skirt_bk_02(ComponentPose=@n3, Alpha=0.5, ExternalForce="0,0,0") {
    angularDampingOverride=0.9999
    angularSpringConstant=75
    bAngularSpring=true
    bGravityOverrideInSimSpace=true
    bOverrideAngularDamping=true
    simulationSpace=RootRelative
    alphaBoolBlend {blendOption=Linear}
    boundBone {boneName=dyn_skirt_bk_02}
    componentLinearAccScale=<1.2 1.2 1.2>
    componentLinearVelScale=<1.2 1.2 1.2>
    physicsBodyDefinitions [{
      sphereCollisionRadius=0
      boundBone {boneName=dyn_skirt_bk_02}
      boxExtents=<20 20 10>
      constraintSetup {angularLimitsMax=<25 0 25>, angularLimitsMin=<-25 -25 -25>, angularTarget=<0 0 0>}
      localJointOffset=<2 0 0>
     }]
   }
   n5 = Anim Dynamics - Bone: dyn_skirt_bk_01_r(ComponentPose=@n4, Alpha=0.5, ExternalForce="0,0,0") {
    angularDampingOverride=0.9999
    angularSpringConstant=75
    bAngularSpring=true
    bGravityOverrideInSimSpace=true
    bOverrideAngularDamping=true
    bUseSphericalLimits=true
    simulationSpace=RootRelative
    alphaBoolBlend {blendOption=Linear}
    boundBone {boneName=dyn_skirt_bk_01_r}
    componentLinearAccScale=<1.2 1.2 1.2>
    componentLinearVelScale=<1.2 1.2 1.2>
    physicsBodyDefinitions [{
      sphereCollisionRadius=0
      boundBone {boneName=dyn_skirt_bk_01_r}
      boxExtents=<20 20 10>
      constraintSetup {angularLimitsMax=<25 0 25>, angularLimitsMin=<-25 -25 -25>, angularTarget=<0 0 0>}
      localJointOffset=<2 0 0>
     }]
    sphericalLimits [{limitRadius=20, drivingBone {boneName=thigh_r}, sphereLocalOffset=<22.5669 0 0>}]
   }
   n6 = Anim Dynamics - Bone: dyn_skirt_bk_02_r(ComponentPose=@n5, Alpha=0.5, ExternalForce="0,0,0") {
    angularDampingOverride=0.9999
    angularSpringConstant=75
    bAngularSpring=true
    bGravityOverrideInSimSpace=true
    bOverrideAngularDamping=true
    bUseSphericalLimits=true
    simulationSpace=RootRelative
    alphaBoolBlend {blendOption=Linear}
    boundBone {boneName=dyn_skirt_bk_02_r}
    componentLinearAccScale=<1.2 1.2 1.2>
    componentLinearVelScale=<1.2 1.2 1.2>
    physicsBodyDefinitions [{
      sphereCollisionRadius=0
      boundBone {boneName=dyn_skirt_bk_02_r}
      boxExtents=<20 20 10>
      constraintSetup {angularLimitsMax=<25 0 25>, angularLimitsMin=<-25 -25 -25>, angularTarget=<0 0 0>}
      localJointOffset=<2 0 0>
     }]
    sphericalLimits [{limitRadius=20, drivingBone {boneName=thigh_r}, sphereLocalOffset=<23.4498 0 0>}]
   }
   n7 = Anim Dynamics - Bone: dyn_skirt_01_l(ComponentPose=@n6, Alpha=0.5, ExternalForce="0,0,0") {
    angularDampingOverride=0.9999
    angularSpringConstant=75
    bAngularSpring=true
    bGravityOverrideInSimSpace=true
    bOverrideAngularDamping=true
    bUseSphericalLimits=true
    simulationSpace=RootRelative
    alphaBoolBlend {blendOption=Linear}
    boundBone {boneName=dyn_skirt_01_l}
    componentLinearAccScale=<1.2 1.2 1.2>
    componentLinearVelScale=<1.2 1.2 1.2>
    physicsBodyDefinitions [{
      sphereCollisionRadius=0
      boundBone {boneName=dyn_skirt_01_l}
      boxExtents=<20 20 10>
      constraintSetup {angularLimitsMax=<0 0 25>, angularLimitsMin=<-25 -25 -25>, angularTarget=<0 0 0>}
      localJointOffset=<2 0 0>
     }]
    sphericalLimits [{limitRadius=20, drivingBone {boneName=thigh_l}, sphereLocalOffset=<-23.5932 0 0>}]
   }
   n8 = Anim Dynamics - Bone: dyn_skirt_02_l(ComponentPose=@n7, Alpha=0.5, ExternalForce="0,0,0") {
    angularDampingOverride=0.9999
    angularSpringConstant=75
    bAngularSpring=true
    bGravityOverrideInSimSpace=true
    bOverrideAngularDamping=true
    bUseSphericalLimits=true
    simulationSpace=RootRelative
    alphaBoolBlend {blendOption=Linear}
    boundBone {boneName=dyn_skirt_02_l}
    componentLinearAccScale=<1.2 1.2 1.2>
    componentLinearVelScale=<1.2 1.2 1.2>
    physicsBodyDefinitions [{
      sphereCollisionRadius=0
      boundBone {boneName=dyn_skirt_02_l}
      boxExtents=<20 20 10>
      constraintSetup {angularLimitsMax=<0 0 25>, angularLimitsMin=<-25 -25 -25>, angularTarget=<0 0 0>}
      localJointOffset=<2 0 0>
     }]
    sphericalLimits [{limitRadius=20, drivingBone {boneName=thigh_l}, sphereLocalOffset=<-24.392 0 0>}]
   }
   n9 = Anim Dynamics - Bone: dyn_skirt_01_r(ComponentPose=@n8, Alpha=0.5, ExternalForce="0,0,0") {
    angularDampingOverride=0.9999
    angularSpringConstant=75
    bAngularSpring=true
    bGravityOverrideInSimSpace=true
    bOverrideAngularDamping=true
    bUseSphericalLimits=true
    simulationSpace=RootRelative
    alphaBoolBlend {blendOption=Linear}
    boundBone {boneName=dyn_skirt_01_r}
    componentLinearAccScale=<1.2 1.2 1.2>
    componentLinearVelScale=<1.2 1.2 1.2>
    physicsBodyDefinitions [{
      sphereCollisionRadius=0
      boundBone {boneName=dyn_skirt_01_r}
      boxExtents=<20 20 10>
      constraintSetup {angularLimitsMax=<0 0 25>, angularLimitsMin=<-25 -25 -25>, angularTarget=<0 0 0>}
      localJointOffset=<2 0 0>
     }]
    sphericalLimits [{limitRadius=20, drivingBone {boneName=thigh_r}, sphereLocalOffset=<23.6309 0 0>}]
   }
   n10 = Anim Dynamics - Bone: dyn_skirt_02_r(ComponentPose=@n9, Alpha=0.5, ExternalForce="0,0,0") {
    angularDampingOverride=0.9999
    angularSpringConstant=75
    bAngularSpring=true
    bGravityOverrideInSimSpace=true
    bOverrideAngularDamping=true
    bUseSphericalLimits=true
    simulationSpace=RootRelative
    alphaBoolBlend {blendOption=Linear}
    boundBone {boneName=dyn_skirt_02_r}
    componentLinearAccScale=<1.2 1.2 1.2>
    componentLinearVelScale=<1.2 1.2 1.2>
    physicsBodyDefinitions [{
      sphereCollisionRadius=0
      boundBone {boneName=dyn_skirt_02_r}
      boxExtents=<20 20 10>
      constraintSetup {angularLimitsMax=<0 0 25>, angularLimitsMin=<-25 -25 -25>, angularTarget=<0 0 0>}
      localJointOffset=<2 0 0>
     }]
    sphericalLimits [{limitRadius=20, drivingBone {boneName=thigh_r}, sphereLocalOffset=<23.441 0 0>}]
   }
   n11 = "Transform (Modify) Bone - Bone: dyn_necklace_01"!(ComponentPose=@n10, bAlphaBoolEnabled:bool=True, Translation="0,1,1", Rotation="0,0,0", Scale="1,1,1") -> (Pose) {
    alphaInputType=Bool
    rotationMode=BMM_Additive
    rotationSpace=BCS_BoneSpace
    translationMode=BMM_Additive
    translationSpace=BCS_BoneSpace
    alphaBoolBlend {blendOption=Linear}
    boneToModify {boneName=dyn_necklace_01}
    translation {y=1, z=1}
   }
   n12 = RigidBody!(
    ComponentPose=Local To Component(LocalPose=Use cached pose 'SavedPose13862'())
    Alpha
    OverrideWorldGravity:Vector=Get Override World Gravity() -> (Override World Gravity:Vector)
    ComponentLinearAccScale:Vector=Get Component Linear Acc Scale() -> (Component Linear Acc Scale:Vector)
    ComponentLinearVelScale:Vector=Get Component Linear Vel Scale() -> (Component Linear Vel Scale:Vector)
   ) -> (Pose) {
    bClampLinearTranslationLimitToRefPose=true
    bOverrideWorldGravity=true
    overridePhysicsAsset=&/F_MED_KnightCatRacket_Physics
    simulationSpace=BaseBoneSpace
    alphaBoolBlend {blendOption=Linear}
    baseBoneRef {boneName=pelvis}
    componentAppliedLinearAccClamp=<2000 2000 2000>
    componentLinearAccScale=<0.1 0.1 0.1>
    componentLinearVelScale=<0.1 0.1 0.1>
    overrideWorldGravity {z=-980}
   }
   n13 = Get Crouch_skirt() -> (Crouch_skirt:double)
   n14 = "Constraint - dyn_skirt_fr_01_l_MSTR (PP)"!(ComponentPose=@n12, Alpha=@n13, ConstraintWeights_0=0, ConstraintWeights_1:float=1) -> (Pose) {
    alphaBoolBlend {blendOption=Linear}
    alphaScaleBiasClamp {bInterpResult=true, interpSpeedDecreasing=5, interpSpeedIncreasing=5}
    boneToModify {boneName=dyn_skirt_fr_01_l_MSTR}
    constraintSetup [{targetBone {boneName=pelvis}}, {targetBone {boneName=dyn_thigh_01_l}}]
    constraintWeights [1,1]
   }
   n15 = "Constraint - dyn_skirt_fr_01_l (PP)"!(ComponentPose=@n14, Alpha=@n13, ConstraintWeights_0=0, ConstraintWeights_1:float=1) -> (Pose) {
    alphaBoolBlend {blendOption=Linear}
    alphaScaleBiasClamp {bInterpResult=true, interpSpeedDecreasing=5, interpSpeedIncreasing=5}
    boneToModify {boneName=dyn_skirt_fr_01_l}
    constraintSetup [{targetBone {boneName=pelvis}}, {targetBone {boneName=dyn_thigh_01_l}}]
    constraintWeights [1,1]
   }
   n16 = "Constraint - dyn_skirt_fr_02_l (PP)"!(ComponentPose=@n15, Alpha=@n13, ConstraintWeights_0=0, ConstraintWeights_1:float=1) -> (Pose) {
    alphaBoolBlend {blendOption=Linear}
    alphaScaleBiasClamp {bInterpResult=true, interpSpeedDecreasing=5, interpSpeedIncreasing=5}
    boneToModify {boneName=dyn_skirt_fr_02_l}
    constraintSetup [{targetBone {boneName=pelvis}}, {targetBone {boneName=dyn_thigh_01_l}}]
    constraintWeights [1,1]
   }
   n17 = "Constraint - dyn_skirt_fr_01_r_MSTR (PP)"!(ComponentPose=@n16, Alpha=@n13, ConstraintWeights_0=0, ConstraintWeights_1:float=1) -> (Pose) {
    alphaBoolBlend {blendOption=Linear}
    alphaScaleBiasClamp {bInterpResult=true, interpSpeedDecreasing=5, interpSpeedIncreasing=5}
    boneToModify {boneName=dyn_skirt_fr_01_r_MSTR}
    constraintSetup [{targetBone {boneName=pelvis}}, {targetBone {boneName=thigh_twist_01_r}}]
    constraintWeights [1,1]
   }
   n18 = "Constraint - dyn_skirt_fr_01_r (PP)"!(ComponentPose=@n17, Alpha=@n13, ConstraintWeights_0=0, ConstraintWeights_1:float=1) -> (Pose) {
    alphaBoolBlend {blendOption=Linear}
    alphaScaleBiasClamp {bInterpResult=true, interpSpeedDecreasing=5, interpSpeedIncreasing=5}
    boneToModify {boneName=dyn_skirt_fr_01_r}
    constraintSetup [{targetBone {boneName=pelvis}}, {targetBone {boneName=thigh_twist_01_r}}]
    constraintWeights [1,1]
   }
   n19 = "Constraint - dyn_skirt_fr_02_r (PP)"!(ComponentPose=@n18, Alpha=@n13, ConstraintWeights_0=0.1, ConstraintWeights_1:float=0.9) -> (Pose) {
    alphaBoolBlend {blendOption=Linear}
    alphaScaleBiasClamp {bInterpResult=true, interpSpeedDecreasing=5, interpSpeedIncreasing=5}
    boneToModify {boneName=dyn_skirt_fr_02_r}
    constraintSetup [{targetBone {boneName=pelvis}}, {targetBone {boneName=thigh_twist_01_r}}]
    constraintWeights [1,1]
   }
   n20 = "Constraint - dyn_skirt_fr2_01_l_MSTR (PP)"!(ComponentPose=@n19, Alpha=@n13, ConstraintWeights_0=0, ConstraintWeights_1:float=1) -> (Pose) {
    alphaBoolBlend {blendOption=Linear}
    alphaScaleBiasClamp {bInterpResult=true, interpSpeedDecreasing=5, interpSpeedIncreasing=5}
    boneToModify {boneName=dyn_skirt_fr2_01_l_MSTR}
    constraintSetup [{targetBone {boneName=pelvis}}, {targetBone {boneName=dyn_thigh_01_l}}]
    constraintWeights [1,1]
   }
   n21 = "Constraint - dyn_skirt_fr2_01_l (PP)"!(ComponentPose=@n20, Alpha=@n13, ConstraintWeights_0=0, ConstraintWeights_1:float=1) -> (Pose) {
    alphaBoolBlend {blendOption=Linear}
    alphaScaleBiasClamp {bInterpResult=true, interpSpeedDecreasing=5, interpSpeedIncreasing=5}
    boneToModify {boneName=dyn_skirt_fr2_01_l}
    constraintSetup [{targetBone {boneName=pelvis}}, {targetBone {boneName=dyn_thigh_01_l}}]
    constraintWeights [1,1]
   }
   n22 = "Constraint - dyn_skirt_fr2_02_l (PP)"!(ComponentPose=@n21, Alpha=@n13, ConstraintWeights_0=0.1, ConstraintWeights_1:float=0.9) -> (Pose) {
    alphaBoolBlend {blendOption=Linear}
    alphaScaleBiasClamp {bInterpResult=true, interpSpeedDecreasing=5, interpSpeedIncreasing=5}
    boneToModify {boneName=dyn_skirt_fr2_02_l}
    constraintSetup [{targetBone {boneName=pelvis}}, {targetBone {boneName=dyn_thigh_01_l}}]
    constraintWeights [1,1]
   }
   n23 = "Constraint - dyn_skirt_fr2_01_r_MSTR (PP)"!(ComponentPose=@n22, Alpha=@n13, ConstraintWeights_0=0, ConstraintWeights_1:float=1) -> (Pose) {
    alphaBoolBlend {blendOption=Linear}
    alphaScaleBiasClamp {bInterpResult=true, interpSpeedDecreasing=5, interpSpeedIncreasing=5}
    boneToModify {boneName=dyn_skirt_fr2_01_r_MSTR}
    constraintSetup [{targetBone {boneName=pelvis}}, {targetBone {boneName=dyn_thigh_01_r}}]
    constraintWeights [1,1]
   }
   n24 = "Constraint - dyn_skirt_fr2_01_r (PP)"!(ComponentPose=@n23, Alpha=@n13, ConstraintWeights_0=0, ConstraintWeights_1:float=1) -> (Pose) {
    alphaBoolBlend {blendOption=Linear}
    alphaScaleBiasClamp {bInterpResult=true, interpSpeedDecreasing=5, interpSpeedIncreasing=5}
    boneToModify {boneName=dyn_skirt_fr2_01_r}
    constraintSetup [{targetBone {boneName=pelvis}}, {targetBone {boneName=dyn_thigh_01_r}}]
    constraintWeights [1,1]
   }
   n25 = "Constraint - dyn_skirt_fr2_02_r (PP)"!(ComponentPose=@n24, Alpha=@n13, ConstraintWeights_0=0, ConstraintWeights_1:float=1) -> (Pose) {
    alphaBoolBlend {blendOption=Linear}
    alphaScaleBiasClamp {bInterpResult=true, interpSpeedDecreasing=5, interpSpeedIncreasing=5}
    boneToModify {boneName=dyn_skirt_fr2_02_r}
    constraintSetup [{targetBone {boneName=pelvis}}, {targetBone {boneName=dyn_thigh_01_r}}]
    constraintWeights [1,1]
   }
   n26 = Get skirt_on_Moving_fix() -> (skirt_on_Moving_fix:double)
   n27 = "Constraint - dyn_skirt_fr_01_MSTR (PP)"!(ComponentPose=@n25, Alpha=@n26, ConstraintWeights_0=0.5, ConstraintWeights_1:float=0.5) -> (Pose) {
    alphaBoolBlend {blendOption=Linear}
    alphaScaleBiasClamp {bInterpResult=true, interpSpeedDecreasing=5, interpSpeedIncreasing=5}
    boneToModify {boneName=dyn_skirt_fr_01_MSTR}
    constraintSetup [{targetBone {boneName=dyn_skirt_fr_01_l}}, {targetBone {boneName=dyn_skirt_fr_01_r}}]
    constraintWeights [1,1]
   }
   n28 = "Constraint - dyn_skirt_fr_01 (PP)"!(ComponentPose=@n27, Alpha=@n26, ConstraintWeights_0=0.5, ConstraintWeights_1:float=0.5) -> (Pose) {
    alphaBoolBlend {blendOption=Linear}
    alphaScaleBiasClamp {bInterpResult=true, interpSpeedDecreasing=5, interpSpeedIncreasing=5}
    boneToModify {boneName=dyn_skirt_fr_01}
    constraintSetup [{targetBone {boneName=dyn_skirt_fr_01_l}}, {targetBone {boneName=dyn_skirt_fr_01_r}}]
    constraintWeights [1,1]
   }
   n29 = "Constraint - dyn_skirt_fr_02 (PP)"!(ComponentPose=@n28, Alpha=@n26, ConstraintWeights_0=0.5, ConstraintWeights_1:float=0.5) -> (Pose) {
    alphaBoolBlend {blendOption=Linear}
    alphaScaleBiasClamp {bInterpResult=true, interpSpeedDecreasing=5, interpSpeedIncreasing=5}
    boneToModify {boneName=dyn_skirt_fr_02}
    constraintSetup [{targetBone {boneName=dyn_skirt_fr_02_l}}, {targetBone {boneName=dyn_skirt_fr_02_r}}]
    constraintWeights [1,1]
   }
   n30 = Get Skirt_clippingFix_onCrouching() -> (Skirt_clippingFix_onCrouching:double)
   n31 = "Transform (Modify) Bone - Bone: dyn_skirt_fr_01_MSTR"(ComponentPose=@n29, Alpha=@n30, Translation="3,0,0", Rotation="0,0,0", Scale="1,1,1") {
    rotationMode=BMM_Additive
    rotationSpace=BCS_BoneSpace
    translationMode=BMM_Additive
    translationSpace=BCS_BoneSpace
    alphaBoolBlend {blendOption=Linear}
    boneToModify {boneName=dyn_skirt_fr_01_MSTR}
    translation {x=3}
   }
   n32 = "Transform (Modify) Bone - Bone: dyn_thigh_01_l"(ComponentPose=@n31, Alpha=@n30, Translation="-3,0,0", Rotation="0,0,0", Scale="1,1,1") {
    rotationMode=BMM_Additive
    rotationSpace=BCS_BoneSpace
    translationMode=BMM_Additive
    translationSpace=BCS_BoneSpace
    alphaBoolBlend {blendOption=Linear}
    boneToModify {boneName=dyn_thigh_01_l}
    translation {x=-3}
   }
   n33 = "Transform (Modify) Bone - Bone: dyn_skirt_fr_01_MSTR"(ComponentPose=@n32, Alpha=@n26, Translation="0,0,0", Rotation="24,0,0", Scale="1,1,1") {
    rotationMode=BMM_Additive
    rotationSpace=BCS_BoneSpace
    translationMode=BMM_Additive
    translationSpace=BCS_BoneSpace
    alphaBoolBlend {blendOption=Linear}
    alphaScaleBiasClamp {bInterpResult=true, interpSpeedDecreasing=5, interpSpeedIncreasing=5}
    boneToModify {boneName=dyn_skirt_fr_01_MSTR}
    rotation {pitch=24}
   }
   n34 = Get Skirt_L_Crouch_left() -> (Skirt_L_Crouch_left:double)
   n35 = "Transform (Modify) Bone - Bone: dyn_skirt_01_l_MSTR"(ComponentPose=@n33, Alpha=@n34, Translation="0,0,0", Rotation="22,0,0", Scale="1,1,1") {
    rotationMode=BMM_Additive
    rotationSpace=BCS_BoneSpace
    translationMode=BMM_Additive
    translationSpace=BCS_BoneSpace
    alphaBoolBlend {blendOption=Linear}
    boneToModify {boneName=dyn_skirt_01_l_MSTR}
    rotation {pitch=22}
   }
   n36 = "Transform (Modify) Bone - Bone: dyn_skirt_fr2_01_l_MSTR"(ComponentPose=@n35, Alpha=@n34, Translation="0,0,0", Rotation="22,0,0", Scale="1,1,1") {
    rotationMode=BMM_Additive
    rotationSpace=BCS_BoneSpace
    translationMode=BMM_Additive
    translationSpace=BCS_BoneSpace
    alphaBoolBlend {blendOption=Linear}
    boneToModify {boneName=dyn_skirt_fr2_01_l_MSTR}
    rotation {pitch=22}
   }
   n37 = "Transform (Modify) Bone - Bone: dyn_skirt_fr_01_MSTR"(ComponentPose=@n36, Alpha=Get skirt_on_Moving_bkw() -> (skirt_on_Moving_bkw:double), Translation="0,0,0", Rotation="-25,0,0", Scale="1,1,1") {
    rotationMode=BMM_Additive
    rotationSpace=BCS_BoneSpace
    translationMode=BMM_Additive
    translationSpace=BCS_BoneSpace
    alphaBoolBlend {blendOption=Linear}
    alphaScaleBiasClamp {bInterpResult=true, interpSpeedDecreasing=5, interpSpeedIncreasing=5}
    boneToModify {boneName=dyn_skirt_fr_01_MSTR}
    rotation {pitch=-25}
   }
   n38 = Get Alpha_Rotation_History() -> (Alpha_Rotation_History:double)
   n39 = "Transform (Modify) Bone - Bone: dyn_skirt_bk_02_l"(ComponentPose=@n37, Alpha=@n38, Translation="0,0,0", Rotation="35,0,0", Scale="1,1,1") {
    rotationMode=BMM_Additive
    rotationSpace=BCS_BoneSpace
    alphaBoolBlend {blendOption=Linear}
    alphaScaleBiasClamp {bInterpResult=true, interpSpeedDecreasing=5}
    boneToModify {boneName=dyn_skirt_bk_02_l}
    rotation {pitch=35}
   }
   n40 = "Transform (Modify) Bone - Bone: dyn_skirt_bk_02_r"(ComponentPose=@n39, Alpha=@n38, Translation="0,0,0", Rotation="30,0,0", Scale="1,1,1") {
    rotationMode=BMM_Additive
    rotationSpace=BCS_BoneSpace
    alphaBoolBlend {blendOption=Linear}
    alphaScaleBiasClamp {bInterpResult=true, interpSpeedDecreasing=5}
    boneToModify {boneName=dyn_skirt_bk_02_r}
    rotation {pitch=30}
   }
   n41 = "Transform (Modify) Bone - Bone: dyn_skirt_01_r_MSTR"(ComponentPose=@n40, Alpha=@n38, Translation="0,0,0", Rotation="20,0,0", Scale="1,1,1") {
    rotationMode=BMM_Additive
    rotationSpace=BCS_BoneSpace
    alphaBoolBlend {blendOption=Linear}
    alphaScaleBiasClamp {bInterpResult=true, interpSpeedDecreasing=5}
    boneToModify {boneName=dyn_skirt_01_r_MSTR}
    rotation {pitch=20}
   }
   n42 = "Transform (Modify) Bone - Bone: dyn_skirt_bk_02"(ComponentPose=@n41, Alpha=@n38, Translation="0,0,0", Rotation="30,0,0", Scale="1,1,1") {
    rotationMode=BMM_Additive
    rotationSpace=BCS_BoneSpace
    alphaBoolBlend {blendOption=Linear}
    alphaScaleBiasClamp {bInterpResult=true, interpSpeedDecreasing=5}
    boneToModify {boneName=dyn_skirt_bk_02}
    rotation {pitch=30}
   }
   n43 = "Transform (Modify) Bone - Bone: dyn_skirt_02_l"(ComponentPose=@n42, Alpha=@n38, Translation="0,0,0", Rotation="40,0,0", Scale="1,1,1") {
    rotationMode=BMM_Additive
    rotationSpace=BCS_BoneSpace
    alphaBoolBlend {blendOption=Linear}
    alphaScaleBiasClamp {bInterpResult=true, interpSpeedDecreasing=5}
    boneToModify {boneName=dyn_skirt_02_l}
    rotation {pitch=40}
   }
   n44 = "Transform (Modify) Bone - Bone: dyn_skirt_fr_02_l"(ComponentPose=@n43, Alpha=@n38, Translation="0,0,0", Rotation="20,0,0", Scale="1,1,1") {
    rotationMode=BMM_Additive
    rotationSpace=BCS_BoneSpace
    alphaBoolBlend {blendOption=Linear}
    alphaScaleBiasClamp {bInterpResult=true, interpSpeedDecreasing=5}
    boneToModify {boneName=dyn_skirt_fr_02_l}
    rotation {pitch=20}
   }
   n45 = "Transform (Modify) Bone - Bone: dyn_skirt_fr2_02_l"(ComponentPose=@n44, Alpha=@n38, Translation="0,0,0", Rotation="20,0,0", Scale="1,1,1") {
    rotationMode=BMM_Additive
    rotationSpace=BCS_BoneSpace
    alphaBoolBlend {blendOption=Linear}
    alphaScaleBiasClamp {bInterpResult=true, interpSpeedDecreasing=5}
    boneToModify {boneName=dyn_skirt_fr2_02_l}
    rotation {pitch=20}
   }
   n46 = "Transform (Modify) Bone - Bone: dyn_skirt_fr2_02_r"(ComponentPose=@n45, Alpha=@n38, Translation="0,0,0", Rotation="30,0,0", Scale="1,1,1") {
    rotationMode=BMM_Additive
    rotationSpace=BCS_BoneSpace
    alphaBoolBlend {blendOption=Linear}
    alphaScaleBiasClamp {bInterpResult=true, interpSpeedDecreasing=5}
    boneToModify {boneName=dyn_skirt_fr2_02_r}
    rotation {pitch=30}
   }
   n47 = "Transform (Modify) Bone - Bone: dyn_skirt_fr_02_r"(ComponentPose=@n46, Alpha=@n38, Translation="0,0,0", Rotation="30,0,0", Scale="1,1,1") {
    rotationMode=BMM_Additive
    rotationSpace=BCS_BoneSpace
    alphaBoolBlend {blendOption=Linear}
    alphaScaleBiasClamp {bInterpResult=true, interpSpeedDecreasing=5}
    boneToModify {boneName=dyn_skirt_fr_02_r}
    rotation {pitch=30}
   }
   n48 = "Transform (Modify) Bone - Bone: dyn_skirt_fr_02"(ComponentPose=@n47, Alpha=@n38, Translation="0,0,0", Rotation="40,0,0", Scale="1,1,1") {
    rotationMode=BMM_Additive
    rotationSpace=BCS_BoneSpace
    alphaBoolBlend {blendOption=Linear}
    alphaScaleBiasClamp {bInterpResult=true, interpSpeedDecreasing=5}
    boneToModify {boneName=dyn_skirt_fr_02}
    rotation {pitch=40}
   }
   n49 = Blend Poses by bool(
    BlendPose_0=Component To Local(ComponentPose=@n11)
    BlendPose_1=Component To Local(ComponentPose=@n48)
    BlendTime_0=0.01
    BlendTime_1=0.01
    bActiveValue=Get SkyDive_mode() -> (SkyDive_mode:bool)
   ) {blendType=Linear}
   n50 = Anim Dynamics - Bone: dyn_skirt_fr2_01_r(ComponentPose=Local To Component(LocalPose=Use cached pose 'SavedPose13862'()), Alpha=1, ExternalForce="0,0,0") {
    angularDampingOverride=0.9999
    angularSpringConstant=75
    bAngularSpring=true
    bGravityOverrideInSimSpace=true
    bOverrideAngularDamping=true
    bUseGravityOverride=true
    bUsePlanarLimit=false
    simulationSpace=RootRelative
    alphaBoolBlend {blendOption=Linear}
    boundBone {boneName=dyn_skirt_fr2_01_r}
    componentLinearAccScale=<1.2 1.2 1.2>
    componentLinearVelScale=<1.2 1.2 1.2>
    gravityOverride {z=-980}
    physicsBodyDefinitions [{
      sphereCollisionRadius=0
      boundBone {boneName=dyn_skirt_fr2_01_r}
      boxExtents=<20 20 10>
      constraintSetup {angularLimitsMax=<0 0 15>, angularLimitsMin=<-25 -60 -15>, angularTarget=<0 0 0>}
      localJointOffset=<2 0 0>
     }]
   }
   n51 = Anim Dynamics - Bone: dyn_skirt_fr2_02_r(ComponentPose=@n50, Alpha=1, ExternalForce="0,0,0") {
    angularDampingOverride=0.9999
    angularSpringConstant=75
    bAngularSpring=true
    bGravityOverrideInSimSpace=true
    bOverrideAngularDamping=true
    bUseGravityOverride=true
    bUsePlanarLimit=false
    simulationSpace=RootRelative
    alphaBoolBlend {blendOption=Linear}
    boundBone {boneName=dyn_skirt_fr2_02_r}
    componentLinearAccScale=<1.2 1.2 1.2>
    componentLinearVelScale=<1.2 1.2 1.2>
    gravityOverride {z=-980}
    physicsBodyDefinitions [{
      sphereCollisionRadius=0
      boundBone {boneName=dyn_skirt_fr2_02_r}
      boxExtents=<20 20 10>
      constraintSetup {angularLimitsMax=<0 0 15>, angularLimitsMin=<-25 -60 -15>, angularTarget=<0 0 0>}
      localJointOffset=<2 0 0>
     }]
   }
   n52 = Anim Dynamics - Bone: dyn_skirt_fr_01_r(ComponentPose=@n51, Alpha=1, ExternalForce="0,0,0") {
    angularDampingOverride=0.9999
    angularSpringConstant=600
    bAngularSpring=true
    bGravityOverrideInSimSpace=true
    bOverrideAngularDamping=true
    bUseGravityOverride=true
    simulationSpace=RootRelative
    alphaBoolBlend {blendOption=Linear}
    boundBone {boneName=dyn_skirt_fr_01_r}
    componentAppliedLinearAccClamp=<2000 2000 2000>
    componentLinearAccScale=<1.2 1.2 1.2>
    componentLinearVelScale=<1.2 1.2 1.2>
    gravityOverride {z=-980}
    physicsBodyDefinitions [{
      sphereCollisionRadius=0
      boundBone {boneName=dyn_skirt_fr_01_r}
      boxExtents=<20 20 10>
      constraintSetup {angularLimitsMax=<25 60 15>, angularLimitsMin=<-25 -60 -15>, angularTarget=<0 0 0>}
      localJointOffset=<2 0 0>
     }]
    planarLimits [{drivingBone {boneName=thigh_r}, planeTransform {location=<-0.0012 -5.9083 0.3961>, rotation=<rot 0.8164 -12.2861 -86.2528>}}]
   }
   n53 = Anim Dynamics - Bone: dyn_skirt_fr_02_r(ComponentPose=@n52, Alpha=1, ExternalForce="0,0,0") {
    angularDampingOverride=0.9999
    angularSpringConstant=600
    bAngularSpring=true
    bGravityOverrideInSimSpace=true
    bOverrideAngularDamping=true
    bUseGravityOverride=true
    simulationSpace=RootRelative
    alphaBoolBlend {blendOption=Linear}
    boundBone {boneName=dyn_skirt_fr_02_r}
    componentAppliedLinearAccClamp=<2000 2000 2000>
    componentLinearAccScale=<1.2 1.2 1.2>
    componentLinearVelScale=<1.2 1.2 1.2>
    gravityOverride {z=-980}
    physicsBodyDefinitions [{
      sphereCollisionRadius=0
      boundBone {boneName=dyn_skirt_fr_02_r}
      boxExtents=<20 20 10>
      constraintSetup {angularLimitsMax=<25 90 15>, angularLimitsMin=<-25 -90 -15>, angularTarget=<0 0 0>}
      localJointOffset=<2 0 0>
     }]
    planarLimits [{drivingBone {boneName=thigh_r}, planeTransform {location=<-1.0418 -6.5026 -0.256>, rotation=<rot -0.3563 -9.0882 -92.2265>}}]
   }
   n54 = Anim Dynamics - Bone: dyn_skirt_fr_01(ComponentPose=@n53, Alpha=1, ExternalForce="0,0,0") {
    angularDampingOverride=0.9999
    angularSpringConstant=75
    bAngularSpring=true
    bGravityOverrideInSimSpace=true
    bOverrideAngularDamping=true
    bUseGravityOverride=true
    bUsePlanarLimit=false
    simulationSpace=RootRelative
    alphaBoolBlend {blendOption=Linear}
    boundBone {boneName=dyn_skirt_fr_01}
    componentLinearAccScale=<1.2 1.2 1.2>
    componentLinearVelScale=<1.2 1.2 1.2>
    gravityOverride {z=-980}
    physicsBodyDefinitions [{
      sphereCollisionRadius=0
      boundBone {boneName=dyn_skirt_fr_01}
      boxExtents=<20 20 10>
      constraintSetup {angularLimitsMax=<15 60 15>, angularLimitsMin=<-15 -60 -15>, angularTarget=<0 0 0>}
      localJointOffset=<2 0 0>
     }]
   }
   n55 = Anim Dynamics - Bone: dyn_skirt_fr_02(ComponentPose=@n54, Alpha=1, ExternalForce="0,0,0") {
    angularDampingOverride=0.9999
    angularSpringConstant=75
    bAngularSpring=true
    bGravityOverrideInSimSpace=true
    bOverrideAngularDamping=true
    bUseGravityOverride=true
    bUsePlanarLimit=false
    simulationSpace=RootRelative
    alphaBoolBlend {blendOption=Linear}
    boundBone {boneName=dyn_skirt_fr_02}
    componentLinearAccScale=<1.2 1.2 1.2>
    componentLinearVelScale=<1.2 1.2 1.2>
    gravityOverride {z=-980}
    physicsBodyDefinitions [{
      sphereCollisionRadius=0
      boundBone {boneName=dyn_skirt_fr_02}
      boxExtents=<20 20 10>
      constraintSetup {angularLimitsMax=<15 90 15>, angularLimitsMin=<-15 -90 -15>, angularTarget=<0 0 0>}
      localJointOffset=<2 0 0>
     }]
   }
   n56 = Anim Dynamics - Bone: dyn_skirt_fr_01_l(ComponentPose=@n55, Alpha=1, ExternalForce="0,0,0") {
    angularDampingOverride=0.9999
    angularSpringConstant=75
    bAngularSpring=true
    bGravityOverrideInSimSpace=true
    bOverrideAngularDamping=true
    bUseGravityOverride=true
    simulationSpace=RootRelative
    alphaBoolBlend {blendOption=Linear}
    boundBone {boneName=dyn_skirt_fr_01_l}
    componentLinearAccScale=<1.2 1.2 1.2>
    componentLinearVelScale=<1.2 1.2 1.2>
    gravityOverride {z=-980}
    physicsBodyDefinitions [{
      sphereCollisionRadius=0
      boundBone {boneName=dyn_skirt_fr_01_l}
      boxExtents=<20 20 10>
      constraintSetup {angularLimitsMax=<15 60 15>, angularLimitsMin=<-15 -60 -15>, angularTarget=<0 0 0>}
      localJointOffset=<2 0 0>
     }]
    planarLimits [{drivingBone {boneName=thigh_l}, planeTransform {location=<0 3.5051 -0.7715>, rotation=<rot 0 0 123.3439>}}]
   }
   n57 = Anim Dynamics - Bone: dyn_skirt_fr_02_l(ComponentPose=@n56, Alpha=1, ExternalForce="0,0,0") {
    angularDampingOverride=0.9999
    angularSpringConstant=75
    bAngularSpring=true
    bGravityOverrideInSimSpace=true
    bOverrideAngularDamping=true
    bUseGravityOverride=true
    bUsePlanarLimit=false
    simulationSpace=RootRelative
    alphaBoolBlend {blendOption=Linear}
    boundBone {boneName=dyn_skirt_fr_02_l}
    componentLinearAccScale=<1.2 1.2 1.2>
    componentLinearVelScale=<1.2 1.2 1.2>
    gravityOverride {z=-980}
    physicsBodyDefinitions [{
      sphereCollisionRadius=0
      boundBone {boneName=dyn_skirt_fr_02_l}
      boxExtents=<20 20 10>
      constraintSetup {angularLimitsMax=<15 0 15>, angularLimitsMin=<-15 -60 -15>, angularTarget=<0 0 0>}
      localJointOffset=<2 0 0>
     }]
   }
   n58 = Anim Dynamics - Bone: dyn_skirt_fr2_01_l(ComponentPose=@n57, Alpha=1, ExternalForce="0,0,0") {
    angularDampingOverride=0.9999
    angularSpringConstant=75
    bAngularSpring=true
    bGravityOverrideInSimSpace=true
    bOverrideAngularDamping=true
    bUseGravityOverride=true
    simulationSpace=RootRelative
    alphaBoolBlend {blendOption=Linear}
    boundBone {boneName=dyn_skirt_fr2_01_l}
    componentLinearAccScale=<1.2 1.2 1.2>
    componentLinearVelScale=<1.2 1.2 1.2>
    gravityOverride {z=-980}
    physicsBodyDefinitions [{
      sphereCollisionRadius=0
      boundBone {boneName=dyn_skirt_fr2_01_l}
      boxExtents=<20 20 10>
      constraintSetup {angularLimitsMax=<15 0 15>, angularLimitsMin=<-15 -25 -15>, angularTarget=<0 0 0>}
      localJointOffset=<2 0 0>
     }]
    planarLimits [{drivingBone {boneName=thigh_l}, planeTransform {location=<3.319 1.2135 -4.206>, rotation=<rot 5.9341 -3.1412 152.0397>}}]
   }
   n59 = Anim Dynamics - Bone: dyn_skirt_fr2_02_l(ComponentPose=@n58, Alpha=1, ExternalForce="0,0,0") {
    angularDampingOverride=0.9999
    angularSpringConstant=75
    bAngularSpring=true
    bGravityOverrideInSimSpace=true
    bOverrideAngularDamping=true
    bUseGravityOverride=true
    bUsePlanarLimit=false
    simulationSpace=RootRelative
    alphaBoolBlend {blendOption=Linear}
    boundBone {boneName=dyn_skirt_fr2_02_l}
    componentLinearAccScale=<1.2 1.2 1.2>
    componentLinearVelScale=<1.2 1.2 1.2>
    gravityOverride {z=-980}
    physicsBodyDefinitions [{
      sphereCollisionRadius=0
      boundBone {boneName=dyn_skirt_fr2_02_l}
      boxExtents=<20 20 10>
      constraintSetup {angularLimitsMax=<15 0 15>, angularLimitsMin=<-15 -25 -15>, angularTarget=<0 0 0>}
      localJointOffset=<2 0 0>
     }]
   }
   n60 = "Constraint - dyn_skirt_fr_01_l_MSTR (PP)"(ComponentPose=@n59, Alpha=1, ConstraintWeights_0=0, ConstraintWeights_1=1) {
    alphaBoolBlend {blendOption=Linear}
    boneToModify {boneName=dyn_skirt_fr_01_l_MSTR}
    constraintSetup [{targetBone {boneName=pelvis}}, {targetBone {boneName=dyn_thigh_01_l}}]
    constraintWeights [1,1]
   }
   n61 = "Constraint - dyn_skirt_fr_01_l (PP)"(ComponentPose=@n60, Alpha=1, ConstraintWeights_0=0, ConstraintWeights_1=1) {
    alphaBoolBlend {blendOption=Linear}
    boneToModify {boneName=dyn_skirt_fr_01_l}
    constraintSetup [{targetBone {boneName=pelvis}}, {targetBone {boneName=dyn_thigh_01_l}}]
    constraintWeights [1,1]
   }
   n62 = "Constraint - dyn_skirt_fr_02_l (PP)"(ComponentPose=@n61, Alpha=1, ConstraintWeights_0=0, ConstraintWeights_1=1) {
    alphaBoolBlend {blendOption=Linear}
    boneToModify {boneName=dyn_skirt_fr_02_l}
    constraintSetup [{targetBone {boneName=pelvis}}, {targetBone {boneName=dyn_thigh_01_l}}]
    constraintWeights [1,1]
   }
   n63 = "Constraint - dyn_skirt_fr_01_r_MSTR (PP)"(ComponentPose=@n62, Alpha=0.5, ConstraintWeights_0=0, ConstraintWeights_1=1) {
    alphaBoolBlend {blendOption=Linear}
    boneToModify {boneName=dyn_skirt_fr_01_r_MSTR}
    constraintSetup [{targetBone {boneName=pelvis}}, {targetBone {boneName=thigh_twist_01_r}}]
    constraintWeights [1,1]
   }
   n64 = "Constraint - dyn_skirt_fr_01_r (PP)"(ComponentPose=@n63, Alpha=0.5, ConstraintWeights_0=0, ConstraintWeights_1=1) {
    alphaBoolBlend {blendOption=Linear}
    boneToModify {boneName=dyn_skirt_fr_01_r}
    constraintSetup [{targetBone {boneName=pelvis}}, {targetBone {boneName=thigh_twist_01_r}}]
    constraintWeights [1,1]
   }
   n65 = "Constraint - dyn_skirt_fr_02_r (PP)"(ComponentPose=@n64, Alpha=0.5, ConstraintWeights_0=0.1, ConstraintWeights_1=0.9) {
    alphaBoolBlend {blendOption=Linear}
    boneToModify {boneName=dyn_skirt_fr_02_r}
    constraintSetup [{targetBone {boneName=pelvis}}, {targetBone {boneName=thigh_twist_01_r}}]
    constraintWeights [1,1]
   }
   n66 = "Constraint - dyn_skirt_fr2_01_l_MSTR (PP)"(ComponentPose=@n65, Alpha=0.9, ConstraintWeights_0=0, ConstraintWeights_1=1) {
    alphaBoolBlend {blendOption=Linear}
    boneToModify {boneName=dyn_skirt_fr2_01_l_MSTR}
    constraintSetup [{targetBone {boneName=pelvis}}, {targetBone {boneName=dyn_thigh_01_l}}]
    constraintWeights [1,1]
   }
   n67 = "Constraint - dyn_skirt_fr2_01_l (PP)"(ComponentPose=@n66, Alpha=0.9, ConstraintWeights_0=0, ConstraintWeights_1=1) {
    alphaBoolBlend {blendOption=Linear}
    boneToModify {boneName=dyn_skirt_fr2_01_l}
    constraintSetup [{targetBone {boneName=pelvis}}, {targetBone {boneName=dyn_thigh_01_l}}]
    constraintWeights [1,1]
   }
   n68 = "Constraint - dyn_skirt_fr2_02_l (PP)"(ComponentPose=@n67, Alpha=0.8, ConstraintWeights_0=0.1, ConstraintWeights_1=0.9) {
    alphaBoolBlend {blendOption=Linear}
    boneToModify {boneName=dyn_skirt_fr2_02_l}
    constraintSetup [{targetBone {boneName=pelvis}}, {targetBone {boneName=dyn_thigh_01_l}}]
    constraintWeights [1,1]
   }
   n69 = "Constraint - dyn_skirt_fr2_01_r_MSTR (PP)"(ComponentPose=@n68, Alpha=0.9, ConstraintWeights_0=0, ConstraintWeights_1=1) {
    alphaBoolBlend {blendOption=Linear}
    boneToModify {boneName=dyn_skirt_fr2_01_r_MSTR}
    constraintSetup [{targetBone {boneName=pelvis}}, {targetBone {boneName=dyn_thigh_01_r}}]
    constraintWeights [1,1]
   }
   n70 = "Constraint - dyn_skirt_fr2_01_r (PP)"(ComponentPose=@n69, Alpha=0.9, ConstraintWeights_0=0, ConstraintWeights_1=1) {
    alphaBoolBlend {blendOption=Linear}
    boneToModify {boneName=dyn_skirt_fr2_01_r}
    constraintSetup [{targetBone {boneName=pelvis}}, {targetBone {boneName=dyn_thigh_01_r}}]
    constraintWeights [1,1]
   }
   n71 = "Constraint - dyn_skirt_fr2_02_r (PP)"(ComponentPose=@n70, Alpha=0.5, ConstraintWeights_0=0, ConstraintWeights_1=1) {
    alphaBoolBlend {blendOption=Linear}
    boneToModify {boneName=dyn_skirt_fr2_02_r}
    constraintSetup [{targetBone {boneName=pelvis}}, {targetBone {boneName=dyn_thigh_01_r}}]
    constraintWeights [1,1]
   }
   n72 = "Constraint - dyn_skirt_fr_01_MSTR (PP)"(ComponentPose=@n71, Alpha=1, ConstraintWeights_0=0.5, ConstraintWeights_1=0.5) {
    alphaBoolBlend {blendOption=Linear}
    boneToModify {boneName=dyn_skirt_fr_01_MSTR}
    constraintSetup [{targetBone {boneName=dyn_skirt_fr_01_l}}, {targetBone {boneName=dyn_skirt_fr_01_r}}]
    constraintWeights [1,1]
   }
   n73 = "Constraint - dyn_skirt_fr_01 (PP)"(ComponentPose=@n72, Alpha=1, ConstraintWeights_0=0.5, ConstraintWeights_1=0.5) {
    alphaBoolBlend {blendOption=Linear}
    boneToModify {boneName=dyn_skirt_fr_01}
    constraintSetup [{targetBone {boneName=dyn_skirt_fr_01_l}}, {targetBone {boneName=dyn_skirt_fr_01_r}}]
    constraintWeights [1,1]
   }
   n74 = "Constraint - dyn_skirt_fr_02 (PP)"(ComponentPose=@n73, Alpha=1, ConstraintWeights_0=0.5, ConstraintWeights_1=0.5) {
    alphaBoolBlend {blendOption=Linear}
    boneToModify {boneName=dyn_skirt_fr_02}
    constraintSetup [{targetBone {boneName=dyn_skirt_fr_02_l}}, {targetBone {boneName=dyn_skirt_fr_02_r}}]
    constraintWeights [1,1]
   }
   n75 = Anim Dynamics - Bone: dyn_skirt_bk_01_l(ComponentPose=@n74, Alpha=1, ExternalForce="0,0,0") {
    angularDampingOverride=0.9999
    angularSpringConstant=75
    bAngularSpring=true
    bGravityOverrideInSimSpace=true
    bOverrideAngularDamping=true
    bUseGravityOverride=true
    bUsePlanarLimit=false
    bUseSphericalLimits=true
    simulationSpace=BoneRelative
    alphaBoolBlend {blendOption=Linear}
    boundBone {boneName=dyn_skirt_bk_01_l}
    componentAppliedLinearAccClamp=<2000 2000 2000>
    componentLinearAccScale=<1.2 1.2 1.2>
    componentLinearVelScale=<1.2 1.2 1.2>
    gravityOverride {z=-980}
    physicsBodyDefinitions [{
      sphereCollisionRadius=0
      boundBone {boneName=dyn_skirt_bk_01_l}
      boxExtents=<20 20 10>
      constraintSetup {angularLimitsMax=<15 0 10>, angularLimitsMin=<-15 -50 -10>, angularTarget=<0 0 0>}
      localJointOffset=<2 0 0>
     }]
    sphericalLimits [{limitRadius=10, drivingBone {boneName=pelvis}, sphereLocalOffset=<0 0 -3.6524>}]
   }
   n76 = Anim Dynamics - Bone: dyn_skirt_bk_01(ComponentPose=@n75, Alpha=1, ExternalForce="0,0,0") {
    angularDampingOverride=0.9999
    angularSpringConstant=75
    bAngularSpring=true
    bGravityOverrideInSimSpace=true
    bOverrideAngularDamping=true
    bUseGravityOverride=true
    bUsePlanarLimit=false
    bUseSphericalLimits=true
    simulationSpace=BoneRelative
    alphaBoolBlend {blendOption=Linear}
    boundBone {boneName=dyn_skirt_bk_01}
    componentAppliedLinearAccClamp=<10000 10000 10000>
    componentLinearAccScale=<1.2 1.2 1.2>
    componentLinearVelScale=<1.2 1.2 1.2>
    gravityOverride {z=-980}
    physicsBodyDefinitions [{
      sphereCollisionRadius=0
      boundBone {boneName=dyn_skirt_bk_01}
      boxExtents=<20 20 10>
      constraintSetup {angularLimitsMax=<15 15 10>, angularLimitsMin=<-15 -25 -10>, angularTarget=<0 0 0>}
      localJointOffset=<2 0 0>
     }]
    sphericalLimits [{limitRadius=10, drivingBone {boneName=pelvis}, sphereLocalOffset=<-3.2939 -0.9054 0>}]
   }
   n77 = Anim Dynamics - Bone: dyn_skirt_bk_01_r(ComponentPose=@n76, Alpha=1, ExternalForce="0,0,0") {
    angularDampingOverride=0.9999
    angularSpringConstant=75
    bAngularSpring=true
    bGravityOverrideInSimSpace=true
    bOverrideAngularDamping=true
    bUseGravityOverride=true
    bUsePlanarLimit=false
    bUseSphericalLimits=true
    simulationSpace=BoneRelative
    alphaBoolBlend {blendOption=Linear}
    boundBone {boneName=dyn_skirt_bk_01_r}
    componentAppliedLinearAccClamp=<10000 10000 10000>
    componentLinearAccScale=<1.2 1.2 1.2>
    componentLinearVelScale=<1.2 1.2 1.2>
    gravityOverride {z=-980}
    physicsBodyDefinitions [{
      sphereCollisionRadius=0
      boundBone {boneName=dyn_skirt_bk_01_r}
      boxExtents=<20 20 10>
      constraintSetup {angularLimitsMax=<15 15 10>, angularLimitsMin=<-15 -25 -10>, angularTarget=<0 0 0>}
      localJointOffset=<2 0 0>
     }]
    sphericalLimits [{limitRadius=10, drivingBone {boneName=pelvis}, sphereLocalOffset=<0 -0.8058 3.7384>}]
   }
   n78 = Anim Dynamics - Bone: dyn_skirt_01_l(ComponentPose=@n77, Alpha=1, ExternalForce="0,0,0") {
    angularDampingOverride=0.9999
    angularSpringConstant=75
    bAngularSpring=true
    bGravityOverrideInSimSpace=true
    bOverrideAngularDamping=true
    bUseGravityOverride=true
    bUseSphericalLimits=true
    simulationSpace=BoneRelative
    alphaBoolBlend {blendOption=Linear}
    boundBone {boneName=dyn_skirt_01_l}
    componentAppliedLinearAccClamp=<10000 10000 10000>
    componentLinearAccScale=<1.2 1.2 1.2>
    componentLinearVelScale=<1.2 1.2 1.2>
    gravityOverride {z=-980}
    physicsBodyDefinitions [{
      sphereCollisionRadius=0
      boundBone {boneName=dyn_skirt_01_l}
      boxExtents=<20 20 10>
      constraintSetup {angularLimitsMax=<15 15 10>, angularLimitsMin=<-15 -25 -10>, angularTarget=<0 0 0>}
      localJointOffset=<2 0 0>
     }]
    sphericalLimits [{limitRadius=10, drivingBone {boneName=thigh_l}, sphereLocalOffset=<-0.5802 0 3.9324>}]
   }
   n79 = Anim Dynamics - Bone: dyn_skirt_01_r(ComponentPose=@n78, Alpha=1, ExternalForce="0,0,0") {
    angularDampingOverride=0.9999
    angularSpringConstant=75
    bAngularSpring=true
    bGravityOverrideInSimSpace=true
    bOverrideAngularDamping=true
    bUseGravityOverride=true
    bUsePlanarLimit=false
    bUseSphericalLimits=true
    simulationSpace=BoneRelative
    alphaBoolBlend {blendOption=Linear}
    boundBone {boneName=dyn_skirt_01_r}
    componentAppliedLinearAccClamp=<10000 10000 10000>
    componentLinearAccScale=<1.2 1.2 1.2>
    componentLinearVelScale=<1.2 1.2 1.2>
    gravityOverride {z=-980}
    physicsBodyDefinitions [{
      sphereCollisionRadius=0
      boundBone {boneName=dyn_skirt_01_r}
      boxExtents=<20 20 10>
      constraintSetup {angularLimitsMax=<15 15 10>, angularLimitsMin=<-15 -25 -10>, angularTarget=<0 0 0>}
      localJointOffset=<2 0 0>
     }]
    sphericalLimits [{limitRadius=10, drivingBone {boneName=pelvis}, sphereLocalOffset=<0 -2.1997 4.2683>}]
   }
   n80 = Anim Dynamics - Bone: dyn_skirt_bk_02_l(ComponentPose=@n79, Alpha=1, ExternalForce="0,0,0") {
    angularDampingOverride=0.9999
    angularSpringConstant=75
    bAngularSpring=true
    bGravityOverrideInSimSpace=true
    bOverrideAngularDamping=true
    bUseGravityOverride=true
    bUsePlanarLimit=false
    simulationSpace=BoneRelative
    alphaBoolBlend {blendOption=Linear}
    boundBone {boneName=dyn_skirt_bk_02_l}
    componentAppliedLinearAccClamp=<2000 2000 2000>
    componentLinearAccScale=<1.2 1.2 1.2>
    componentLinearVelScale=<1.2 1.2 1.2>
    gravityOverride {z=-980}
    physicsBodyDefinitions [{
      sphereCollisionRadius=0
      boundBone {boneName=dyn_skirt_bk_02_l}
      boxExtents=<20 20 10>
      constraintSetup {angularLimitsMax=<15 0 10>, angularLimitsMin=<-15 -15 -10>, angularTarget=<0 0 0>}
      localJointOffset=<2 0 0>
     }]
   }
   n81 = Anim Dynamics - Bone: dyn_skirt_bk_02(ComponentPose=@n80, Alpha=1, ExternalForce="0,0,0") {
    angularDampingOverride=0.9999
    angularSpringConstant=75
    bAngularSpring=true
    bGravityOverrideInSimSpace=true
    bOverrideAngularDamping=true
    bUseGravityOverride=true
    bUsePlanarLimit=false
    simulationSpace=BoneRelative
    alphaBoolBlend {blendOption=Linear}
    boundBone {boneName=dyn_skirt_bk_02}
    componentAppliedLinearAccClamp=<2000 2000 10000>
    componentLinearAccScale=<1.2 1.2 1.2>
    componentLinearVelScale=<1.2 1.2 1.2>
    gravityOverride {z=-980}
    physicsBodyDefinitions [{
      sphereCollisionRadius=0
      boundBone {boneName=dyn_skirt_bk_02}
      boxExtents=<20 20 10>
      constraintSetup {angularLimitsMax=<15 15 10>, angularLimitsMin=<-15 -25 -10>, angularTarget=<0 0 0>}
      localJointOffset=<2 0 0>
     }]
   }
   n82 = Anim Dynamics - Bone: dyn_skirt_bk_02_r(ComponentPose=@n81, Alpha=1, ExternalForce="0,0,0") {
    angularDampingOverride=0.9999
    angularSpringConstant=75
    bAngularSpring=true
    bGravityOverrideInSimSpace=true
    bOverrideAngularDamping=true
    bUseGravityOverride=true
    bUsePlanarLimit=false
    simulationSpace=BoneRelative
    alphaBoolBlend {blendOption=Linear}
    boundBone {boneName=dyn_skirt_bk_02_r}
    componentAppliedLinearAccClamp=<2000 2000 10000>
    componentLinearAccScale=<1.2 1.2 1.2>
    componentLinearVelScale=<1.2 1.2 1.2>
    gravityOverride {z=-980}
    physicsBodyDefinitions [{
      sphereCollisionRadius=0
      boundBone {boneName=dyn_skirt_bk_02_r}
      boxExtents=<20 20 10>
      constraintSetup {angularLimitsMax=<15 15 10>, angularLimitsMin=<-15 -25 -10>, angularTarget=<0 0 0>}
      localJointOffset=<2 0 0>
     }]
   }
   n83 = Anim Dynamics - Bone: dyn_skirt_02_l(ComponentPose=@n82, Alpha=1, ExternalForce="0,0,0") {
    angularDampingOverride=0.9999
    angularSpringConstant=75
    bAngularSpring=true
    bGravityOverrideInSimSpace=true
    bOverrideAngularDamping=true
    bUseGravityOverride=true
    simulationSpace=BoneRelative
    alphaBoolBlend {blendOption=Linear}
    boundBone {boneName=dyn_skirt_02_l}
    componentAppliedLinearAccClamp=<2000 2000 10000>
    componentLinearAccScale=<1.2 1.2 1.2>
    componentLinearVelScale=<1.2 1.2 1.2>
    gravityOverride {z=-980}
    physicsBodyDefinitions [{
      sphereCollisionRadius=0
      boundBone {boneName=dyn_skirt_02_l}
      boxExtents=<20 20 10>
      constraintSetup {angularLimitsMax=<15 15 10>, angularLimitsMin=<-15 -25 -10>, angularTarget=<0 0 0>}
      localJointOffset=<2 0 0>
     }]
   }
   n84 = Anim Dynamics - Bone: dyn_skirt_02_r(ComponentPose=@n83, Alpha=1, ExternalForce="0,0,0") {
    angularDampingOverride=0.9999
    angularSpringConstant=75
    bAngularSpring=true
    bGravityOverrideInSimSpace=true
    bOverrideAngularDamping=true
    bUseGravityOverride=true
    bUseSphericalLimits=true
    simulationSpace=BoneRelative
    alphaBoolBlend {blendOption=Linear}
    boundBone {boneName=dyn_skirt_02_r}
    componentAppliedLinearAccClamp=<2000 2000 10000>
    componentLinearAccScale=<1.2 1.2 1.2>
    componentLinearVelScale=<1.2 1.2 1.2>
    gravityOverride {z=-980}
    physicsBodyDefinitions [{
      sphereCollisionRadius=0
      boundBone {boneName=dyn_skirt_02_r}
      boxExtents=<20 20 10>
      constraintSetup {angularLimitsMax=<15 15 10>, angularLimitsMin=<-15 -25 -10>, angularTarget=<0 0 0>}
      localJointOffset=<2 0 0>
     }]
    sphericalLimits [{limitRadius=10, drivingBone {boneName=pelvis}, sphereLocalOffset=<-12.2351 -2.0287 6.5244>}]
   }
   n85 = "Anim Dynamics - Chain: dyn_necklace_02 -> dyn_necklace_03"(ComponentPose=@n84, Alpha=1, ExternalForce="0,0,0") {
    angularDampingOverride=0.9999
    angularSpringConstant=75
    bAngularSpring=true
    bChain=true
    bGravityOverrideInSimSpace=true
    bOverrideAngularDamping=true
    simulationSpace=RootRelative
    alphaBoolBlend {blendOption=Linear}
    boundBone {boneName=dyn_necklace_02}
    chainEnd {boneName=dyn_necklace_03}
    componentAppliedLinearAccClamp=<2000 2000 2000>
    componentLinearAccScale=<1.2 1.2 1.2>
    componentLinearVelScale=<1.2 1.2 1.2>
    physicsBodyDefinitions [
     {
      sphereCollisionRadius=0
      boundBone {boneName=dyn_necklace_02}
      boxExtents=<10 10 10>
      constraintSetup {angularLimitsMax=<0 0 25>, angularLimitsMin=<-25 -50 0>, angularTarget=<0 0 0>}
      localJointOffset=<2 0 0>
     }
     {
      sphereCollisionRadius=0
      boundBone {boneName=dyn_necklace_03}
      boxExtents=<10 10 10>
      constraintSetup {angularLimitsMax=<0 0 25>, angularLimitsMin=<-25 -50 0>, angularTarget=<0 0 0>}
      localJointOffset=<1 0 0>
     }
    ]
   }
   n86 = Get skirt_on_Moving_fix()
   n87 = "Transform (Modify) Bone - Bone: dyn_skirt_fr_01_r_MSTR"(ComponentPose=@n85, Alpha=@n86, Translation="0,0,0", Rotation="15.000,0.000,0.000", Scale="1,1,1") {
    rotationMode=BMM_Additive
    rotationSpace=BCS_BoneSpace
    translationMode=BMM_Additive
    translationSpace=BCS_BoneSpace
    alphaBoolBlend {blendOption=Linear}
    alphaScaleBiasClamp {bInterpResult=true}
    boneToModify {boneName=dyn_skirt_fr_01_r_MSTR}
    rotation {pitch=15}
   }
   n88 = "Transform (Modify) Bone - Bone: dyn_skirt_fr_01_MSTR"(ComponentPose=@n87, Alpha=@n86, Translation="0,0,0", Rotation="22,1,0", Scale="1,1,1") {
    rotationMode=BMM_Additive
    rotationSpace=BCS_BoneSpace
    translationMode=BMM_Additive
    translationSpace=BCS_BoneSpace
    alphaBoolBlend {blendOption=Linear}
    alphaScaleBiasClamp {bInterpResult=true}
    boneToModify {boneName=dyn_skirt_fr_01_MSTR}
    rotation {pitch=22, yaw=1}
   }
   n89 = Component To Local(ComponentPose=@n88)
   n90 = Get noRigibBody_mode() -> (noRigibBody_mode:bool)
   n91 = Get skirt_on_low_FPS() -> (skirt_on_low_FPS:double)
   n92 = "Transform (Modify) Bone - Bone: dyn_skirt_01_r_MSTR"(
    ComponentPose=Local To Component(LocalPose=Blend Poses by bool(BlendPose_0=@n89, BlendPose_1=@n49, BlendTime_0=0.35, BlendTime_1=0.35, bActiveValue=@n90) {blendType=Linear})
    Alpha=@n91
    Translation="0,0,0"
    Rotation="12,0,0"
    Scale="1,1,1"
   ) {
    rotationMode=BMM_Additive
    rotationSpace=BCS_BoneSpace
    translationMode=BMM_Additive
    translationSpace=BCS_BoneSpace
    alphaBoolBlend {blendInTime=0.3, blendOption=CubicInOut, blendOutTime=0.3}
    alphaScaleBiasClamp {bInterpResult=true, interpSpeedDecreasing=5, interpSpeedIncreasing=5}
    boneToModify {boneName=dyn_skirt_01_r_MSTR}
    rotation {pitch=12}
   }
   n93 = "Transform (Modify) Bone - Bone: dyn_skirt_bk_01_r_MSTR"(ComponentPose=@n92, Alpha=@n91, Translation="0,0,0", Rotation="12,0,0", Scale="1,1,1") {
    rotationMode=BMM_Additive
    rotationSpace=BCS_BoneSpace
    translationMode=BMM_Additive
    translationSpace=BCS_BoneSpace
    alphaBoolBlend {blendInTime=0.3, blendOption=CubicInOut, blendOutTime=0.3}
    alphaScaleBiasClamp {bInterpResult=true, interpSpeedDecreasing=5, interpSpeedIncreasing=5}
    boneToModify {boneName=dyn_skirt_bk_01_r_MSTR}
    rotation {pitch=12}
   }
   n94 = Control Rig(Source=Component To Local(ComponentPose=@n93)) {
    alphaBoolBlend {blendOption=Linear}
    controlRigAssetReference {blueprintRigClass=&/F_MED_KnightCatRacket_Deform_Control_Rig.F_MED_KnightCatRacket_Deform_Control_Rig_C}
   }
   Output Pose(Result=@n94) {layerGroup=None}
   Save cached pose 'SavedPose13862'(Pose=Copy Pose From Mesh() {bUseAttachedParent=true}) {cacheName=SavedPose13862}
  ]}
 EventGraph {nodes [
   n1 = Event BlueprintUpdateAnimation()
   > Set Delta Time X(Delta Time X=@n1.DeltaTimeX)
   > n2 = Gravity Override State Function DEPRECATED!(
    execute
    CharacterPartAnimInstance=Self-Reference()
    GameTimeInSeconds=GetGameTimeInSeconds()
    GravityOverrideParams_Skydive_GoingUp_28_#1_WindFrequency_28_#2:Vector="1.5,7,8"
    GravityOverrideParams_Skydive_GoingUp_28_#1_WindAmplitude_31_#3:Vector="200,-500,400"
    GravityOverrideParams_Skydive_GoingUp_28_#1_WindOffset_32_#4:Vector="0,0,0"
    GravityOverrideParams_Skydive_GoingUp_28_#1_GravityOverride_15_#5:Vector="0,0,-100"
    GravityOverrideParams_Skydive_GoingUp_28_#1_JointName_27_#6:Name=root
    GravityOverrideParams_Skydive_FloatingDown_29_#7_WindFrequency_28_#2:Vector="1.5,7,8"
    GravityOverrideParams_Skydive_FloatingDown_29_#7_WindAmplitude_31_#3:Vector="200,-500,400"
    GravityOverrideParams_Skydive_FloatingDown_29_#7_WindOffset_32_#4:Vector="0,-600,400"
    GravityOverrideParams_Skydive_FloatingDown_29_#7_GravityOverride_15_#5:Vector="0,-700,600"
    GravityOverrideParams_Skydive_FloatingDown_29_#7_JointName_27_#6:Name=root
    GravityOverrideParams_Skydive_DivingDown_30_#8_WindFrequency_28_#2:Vector="1.5,7,8"
    GravityOverrideParams_Skydive_DivingDown_30_#8_WindAmplitude_31_#3:Vector="50,300,100"
    GravityOverrideParams_Skydive_DivingDown_30_#8_WindOffset_32_#4:Vector="0,0,0"
    GravityOverrideParams_Skydive_DivingDown_30_#8_GravityOverride_15_#5:Vector="0,0,0"
    GravityOverrideParams_Skydive_DivingDown_30_#8_JointName_27_#6:Name=pelvis
    GravityOverrideParams_Skydive_Parachuting_31_#9_WindFrequency_28_#2:Vector="1.5,7,8"
    GravityOverrideParams_Skydive_Parachuting_31_#9_WindAmplitude_31_#3:Vector="200,-500,400"
    GravityOverrideParams_Skydive_Parachuting_31_#9_WindOffset_32_#4:Vector="0,0,0"
    GravityOverrideParams_Skydive_Parachuting_31_#9_GravityOverride_15_#5:Vector="0,-800,-980"
    GravityOverrideParams_Skydive_Parachuting_31_#9_JointName_27_#6:Name=root
    GravityOverrideParams_OnGround_Standing_32_#10:Gravity Override Setting Struct Structure=(WindFrequency_28_#2=(X=0,Y=0,Z=0),WindAmplitude_31_#3=(X=0,Y=0,Z=0),WindOffset_32_#4=(X=0,Y=0,Z=0),GravityOverride_15_#5=(X=0,Y=0,Z=-980),JointName_27_#6="root")
    GravityOverrideParams_OnGround_Moving_33_#11:Gravity Override Setting Struct Structure=(WindFrequency_28_#2=(X=0,Y=0,Z=0),WindAmplitude_31_#3=(X=0,Y=0,Z=0),WindOffset_32_#4=(X=0,Y=0,Z=0),GravityOverride_15_#5=(X=0,Y=0,Z=-980),JointName_27_#6="root")
    GravityOverrideParams_EmoteOrMelee_34_#12:Gravity Override Setting Struct Structure=(WindFrequency_28_#2=(X=0,Y=0,Z=0),WindAmplitude_31_#3=(X=0,Y=0,Z=0),WindOffset_32_#4=(X=0,Y=0,Z=0),GravityOverride_15_#5=(X=0,Y=0,Z=-980),JointName_27_#6="root")
    GravityOverrideParams_Swimming_35_#13:Gravity Override Setting Struct Structure=(WindFrequency_28_#2=(X=0,Y=0,Z=0),WindAmplitude_31_#3=(X=0,Y=0,Z=0),WindOffset_32_#4=(X=0,Y=0,Z=0),GravityOverride_15_#5=(X=-800,Y=0,Z=0),JointName_27_#6="pelvis")
    GravityOverrideParams_Front_End_36_#14:Gravity Override Setting Struct Structure=(WindFrequency_28_#2=(X=0,Y=0,Z=0),WindAmplitude_31_#3=(X=0,Y=0,Z=0),WindOffset_32_#4=(X=0,Y=0,Z=0),GravityOverride_15_#5=(X=0,Y=0,Z=-980),JointName_27_#6="root")
    GravityOverrideParams_DBNO_40_#15:Gravity Override Setting Struct Structure=(WindFrequency_28_#2=(X=0,Y=0,Z=0),WindAmplitude_31_#3=(X=0,Y=0,Z=0),WindOffset_32_#4=(X=0,Y=0,Z=0),GravityOverride_15_#5=(X=-980,Y=400,Z=0),JointName_27_#6="pelvis")
    GravityOverrideParams_DBNO_Carried_41_#16:Gravity Override Setting Struct Structure=(WindFrequency_28_#2=(X=0,Y=0,Z=0),WindAmplitude_31_#3=(X=0,Y=0,Z=0),WindOffset_32_#4=(X=0,Y=0,Z=0),GravityOverride_15_#5=(X=-980,Y=500,Z=-500),JointName_27_#6="pelvis")
    GravityOverrideParams_OnGround_Falling_45_#17:Gravity Override Setting Struct Structure=(WindFrequency_28_#2=(X=0,Y=0,Z=0),WindAmplitude_31_#3=(X=0,Y=0,Z=0),WindOffset_32_#4=(X=0,Y=0,Z=0),GravityOverride_15_#5=(X=0,Y=0,Z=-980),JointName_27_#6="root")
    GravityOverrideParams_Ziplining_46_#18:Gravity Override Setting Struct Structure=(WindFrequency_28_#2=(X=0,Y=0,Z=0),WindAmplitude_31_#3=(X=0,Y=0,Z=0),WindOffset_32_#4=(X=0,Y=0,Z=0),GravityOverride_15_#5=(X=0,Y=0,Z=-980),JointName_27_#6="root")
    GravityOverrideParams_Skydive_Surfing_49_#19:Gravity Override Setting Struct Structure=(WindFrequency_28_#2=(X=1.5,Y=7,Z=8),WindAmplitude_31_#3=(X=200,Y=-500,Z=400),WindOffset_32_#4=(X=0,Y=0,Z=0),GravityOverride_15_#5=(X=500,Y=-300,Z=-980),JointName_27_#6="root")
    GravityOverrideParams_Riding_Standing_54_#20:Gravity Override Setting Struct Structure=(WindFrequency_28_#2=(X=0,Y=0,Z=0),WindAmplitude_31_#3=(X=0,Y=0,Z=0),WindOffset_32_#4=(X=0,Y=0,Z=0),GravityOverride_15_#5=(X=0,Y=-350,Z=-980),JointName_27_#6="root")
    GravityOverrideParams_Riding_Moving_55_#21:Gravity Override Setting Struct Structure=(WindFrequency_28_#2=(X=0,Y=0,Z=0),WindAmplitude_31_#3=(X=0,Y=0,Z=0),WindOffset_32_#4=(X=0,Y=0,Z=0),GravityOverride_15_#5=(X=0,Y=-450,Z=-450),JointName_27_#6="root")
    GravityOverrideParams_Riding_Moving_Targeting_58_#22:Gravity Override Setting Struct Structure=(WindFrequency_28_#2=(X=0,Y=0,Z=0),WindAmplitude_31_#3=(X=0,Y=0,Z=0),WindOffset_32_#4=(X=0,Y=0,Z=0),GravityOverride_15_#5=(X=0,Y=-200,Z=-980),JointName_27_#6="root")
    GravityOverrideParams_Tactical_Sprint_57_#23:Gravity Override Setting Struct Structure=(WindFrequency_28_#2=(X=0,Y=0,Z=0),WindAmplitude_31_#3=(X=0,Y=0,Z=0),WindOffset_32_#4=(X=0,Y=0,Z=0),GravityOverride_15_#5=(X=0,Y=-450,Z=-500),JointName_27_#6="root")
    GravityOverrideParams_Rail_Grinding_61_#24:Gravity Override Setting Struct Structure=(WindFrequency_28_#2=(X=0,Y=0,Z=0),WindAmplitude_31_#3=(X=0,Y=0,Z=0),WindOffset_32_#4=(X=0,Y=0,Z=0),GravityOverride_15_#5=(X=0,Y=0,Z=-980),JointName_27_#6="root")
    GravityOverrideParams_DBNO_Tumbling_67_#25:Gravity Override Setting Struct Structure=(WindFrequency_28_#2=(X=0,Y=0,Z=0),WindAmplitude_31_#3=(X=0,Y=0,Z=0),WindOffset_32_#4=(X=0,Y=0,Z=0),GravityOverride_15_#5=(X=-2000,Y=0,Z=0),JointName_27_#6="pelvis")
    GravityOverrideParams_DBNO_RollFwd_68_#26:Gravity Override Setting Struct Structure=(WindFrequency_28_#2=(X=0,Y=0,Z=0),WindAmplitude_31_#3=(X=0,Y=0,Z=0),WindOffset_32_#4=(X=0,Y=0,Z=0),GravityOverride_15_#5=(X=-2000,Y=0,Z=0),JointName_27_#6="pelvis")
    GravityOverrideParams_DBNO_RollBwd_69_#27:Gravity Override Setting Struct Structure=(WindFrequency_28_#2=(X=0,Y=0,Z=0),WindAmplitude_31_#3=(X=0,Y=0,Z=0),WindOffset_32_#4=(X=0,Y=0,Z=0),GravityOverride_15_#5=(X=-4000,Y=0,Z=0),JointName_27_#6="pelvis")
    GravityOverrideParams_DBNO_Impulse_70_#28:Gravity Override Setting Struct Structure=(WindFrequency_28_#2=(X=0,Y=0,Z=0),WindAmplitude_31_#3=(X=0,Y=0,Z=0),WindOffset_32_#4=(X=0,Y=0,Z=0),GravityOverride_15_#5=(X=-2000,Y=0,Z=0),JointName_27_#6="pelvis")
    RBAN Params
    DebugDraw
    DebugJointName
   ) -> (then, GravityOverride, CurrentStateOut, GravityOverrideSettingsOut, RBAN SettingsOut)
   > Set skirt_on_low_FPS(skirt_on_low_FPS:double=MapRangeClamped(Value=@n1.DeltaTimeX, InRangeA=0.013, InRangeB=0.033)) -> (Output_Get:double)
   > Set Override World Gravity(Override World Gravity:Vector=@n2) -> (Output_Get:Vector)
   > Set SkyDive_mode(SkyDive_mode:bool=OR Boolean(A=Get bIsSkyDiving(), B=Get bIsSkydiveDiveMode())) -> (Output_Get:bool)
   > Set noRigibBody_mode(noRigibBody_mode:bool=NOT Boolean(A=Get bRigidBodyEnabled())) -> (Output_Get:bool)
   n3 = Get bIsJumping()
   n4 = Get bIsFalling()
   n5 = AND Boolean(
    A=OR Boolean!(A=Get bIsCrouching(), B=Get bIsCrouchSprinting(), C:bool=Get bIsSprinting()) -> (ReturnValue)
    B=NOT Boolean(A=OR Boolean!(A=@n3, B=@n4) -> (ReturnValue))
   )
   > n6 = Set Crouch_skirt(Crouch_skirt:double=SelectFloat(A=1, B=0, bPickA=@n5)) -> (Output_Get:double)
   n7 = Vector Length(A=Get PawnVelocity())
   > Set skirt_on_Moving_fix(skirt_on_Moving_fix:double=SelectFloat(A=@n6, B=MapRangeClamped(Value=@n7, InRangeA=0, InRangeB=500), bPickA=@n5)) -> (Output_Get:double)
   n8 = OR Boolean(A=Get bIsCrouching(), B=Get bIsCrouchSprinting())
   n9 = Dot Product!(
    A:Vector=GetActorForwardVector!(self:Actor Object Reference=TryGetPawnOwner()) -> (ReturnValue:Vector)
    B:Vector=Normalize!(A:Vector=Get PawnVelocity(), Tolerance:float=0.0001) -> (ReturnValue:Vector) {is=title K2Node_CallFunction}
   ) -> (ReturnValue:double)
   > Set skirt_on_Moving_bkw(skirt_on_Moving_bkw:double=SelectFloat(A=MapRangeClamped(Value=@n9, InRangeA=-0.6, InRangeB=-1), B=0, bPickA=NOT Boolean(A=@n8))) -> (Output_Get:double)
   > Set Skirt_alpha_for_cape(Skirt_alpha_for_cape:double=0) -> (Output_Get:double)
   n10 = SelectFloat(A=1, B=0, bPickA=Get bIsCrouching())
   n11 = Get SkelMesh()
   > Set Skirt_clippingFix_onCrouching(Skirt_clippingFix_onCrouching:double=float * float(
     A=GetAlphaFromReferenceTransformDelta(SkeletalMeshComponent=@n11, BoneName=pelvis, BaseName=thigh_l, InputRange=(X=15,Y=103), TwistAxis="1,0,0")
     B=@n10
    ) {is=title K2Node_PromotableOperator}) -> (Output_Get:double)
   n12 = Get SkelMesh()
   > Set Skirt_L_Crouch_left(Skirt_L_Crouch_left:double=float * float(
     A=GetAlphaFromReferenceTransformDelta(
      SkeletalMeshComponent=@n12
      BoneName=thigh_l
      BaseName=pelvis
      DeltaComponentType=EulerX
      InputRange=(X=3,Y=-13)
      OutputRange=(X=0,Y=0.5)
      TwistAxis="1,0,0"
     )
     B=@n10
    )) -> (Output_Get:double)
   n13 = Get SkelMesh()
   > Set Spine_05_Rotation(Spine_05_Rotation:Vector=MakeVector(Y=GetAlphaFromReferenceTransformDelta(SkeletalMeshComponent=@n13, BoneName=spine_05, BaseName=spine_01, DeltaComponentType=EulerY, InputRange=(X=0,Y=1), TwistAxis="1,0,0").RawOutput)) -> (Output_Get:Vector)
   > n14 = Calculate Velocity From Position History!(
    execute
    DeltaSeconds:float=Get Delta Time X()
    Position:Vector=Get Spine_05_Rotation() -> (Spine_05_Rotation:Vector)
    History:"Position History Structure (by ref)"=Get History() -> (History:Position History Structure)
    NumberOfSamples:int=16
    VelocityMin:float=5
    VelocityMax:float=64
   ) -> (then, ReturnValue:float)
   n15 = SelectFloat(A=0.8, B=1, bPickA=Get bIsPlayingMeleeAnim())
   n16 = OR Boolean!(A=Get bIsJumping(), B=Get bIsFalling(), C:bool=Get bIsSprinting()) -> (ReturnValue)
   n17 = float * float(A=@n14, B="To Float (Boolean)"(InBool=NOT Boolean(A=@n16)))
   > Set Alpha_Rotation_History(Alpha_Rotation_History:double="Clamp (Float)"(Value="Clamp (Float)"(Value=@n17, Min=0, Max=@n15), Min=0, Max=Get Skirt_alpha_for_cape() -> (Skirt_alpha_for_cape:double))) -> (Output_Get:double)
   Event OnCharacterCustomizationLoaded()
   > n18 = Set SkelMesh(SkelMesh=GetOwnerSkeletalMeshComponent())
   > n19 = Get Dynamic Material Instance(SkeletalMeshComponent=@n18)
   > Set MattInstanceDynamic(MattInstanceDynamic:Material Instance Dynamic Object Reference=@n19) -> (Output_Get:Material Instance Dynamic Object Reference)
   > n20 = GetOrCreateDynamicOverlayMaterial(MeshComponent=@n18, MaterialName=None)
   > Set OverlayMatInstanceDynamic(OverlayMatInstanceDynamic:Material Instance Dynamic Object Reference=@n20) -> (Output_Get:Material Instance Dynamic Object Reference)
   > n21 = Set IsWearingCape(IsWearingCape:bool=CheckForBackbling().bIsBackblingACape) -> (Output_Get:bool)
   > n22 = Branch(Condition=@n21)
   > Set Component Linear Acc Scale(Component Linear Acc Scale:Vector="0.3,0.3,0.3") -> (Output_Get:Vector)
   > Set Component Linear Vel Scale(Component Linear Vel Scale:Vector="0.3,0.3,0.3") -> (Output_Get:Vector)
   Set Component Linear Acc Scale(execute=@n22.else, Component Linear Acc Scale="0.4,0.4,0.4")
   > Set Component Linear Vel Scale(Component Linear Vel Scale="0.4,0.4,0.4")
  ]}
]
variables [
 Alpha_Rotation_History {type=number}
 Component Linear Acc Scale {type=Vector, default=<0.1 0.1 0.1>}
 Component Linear Vel Scale {type=Vector, default=<0.1 0.1 0.1>}
 Crouch_skirt {type=number}
 Delta Time X {type=number}
 History {type=PositionHistory, default {range=0, positions []}}
 IsWearingCape {type=boolean}
 MattInstanceDynamic {type=MaterialInstanceDynamic}
 noRigibBody_mode {type=boolean}
 OverlayMatInstanceDynamic {type=MaterialInstanceDynamic}
 Override World Gravity {type=Vector, default=<0 0 -980>}
 SkelMesh {type=SkeletalMeshComponent}
 Skirt_alpha_for_cape {type=number}
 Skirt_clippingFix_onCrouching {type=number}
 skirt_Crouch_fix {type=number}
 Skirt_L_Crouch_left {type=number}
 skirt_on_low_FPS {type=number}
 skirt_on_low_FPS_0 {type=number}
 skirt_on_Moving_bkw {type=number}
 skirt_on_Moving_fix {type=number}
 Skirt_onJumping {type=boolean}
 SkyDive_mode {type=boolean}
 Spine_05_Rotation {type=Vector}
]
