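# Physics asset for the KnightCatRacket female medium character's FaceAcc mesh, rigging its ponytail and loose hair for physics.
# A five-link capsule chain runs from dyn_hair_tail_MSTR, with two strands branching off the second link and back hair joined to the neck, all swing/twist limited.
# Head, neck, upper spine and left-only shoulder/arm capsules provide body volumes with free angular joints.
# Oddities include a clavicle_l-to-tail_2 constraint and a duplicated spine_05-to-spine_04 constraint.
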PhysicsAsset /BRCosmetics/Characters/Player/Female/Medium/Bodies/F_MED_KnightCatRacket/Meshes/Parts/F_MED_KnightCatRacket_FaceAcc_Physics
bodies {
 clavicle_l [{
   shapeType=Capsule
   length=13.4175
   radius=5.6296
   shapeName=clavicle_l_capsule
   center=<6.1367 0 -1.4561>
   rotation=<rot 85.1162 -179.9946 -179.9946>
  }]
 dyn_hair_back [{shapeType=Capsule, length=5.1773, radius=0.8227, shapeName=dyn_hair_back_capsule, center=<0.7744 -3.3892 -1.6449>}]
 dyn_hair_tail_1 [{shapeType=Capsule, length=4.7095, radius=2.641, shapeName=dyn_hair_tail_1_capsule, center=<3.4042 0 0>, rotation=<rot 90 116.8013 97.7875>}]
 dyn_hair_tail_2 [{shapeType=Capsule, length=2.1094, radius=3.1406, shapeName=dyn_hair_tail_2_capsule, center=<3.4042 -0.2016 0>, rotation=<rot 90 0 0>}]
 dyn_hair_tail_3 [{shapeType=Capsule, length=2.7694, radius=3.3605, shapeName=dyn_hair_tail_3_capsule, center=<3.4042 -0.5306 0>, rotation=<rot 90 0 0>}]
 dyn_hair_tail_4 [{
   shapeType=Capsule
   length=6.6456
   radius=2.8696
   shapeName=dyn_hair_tail_4_capsule
   center=<2.785 -0.0111 0.6366>
   rotation=<rot 0 -179.9998 -158.6242>
  }]
 dyn_hair_tail_5 [{shapeType=Capsule, length=5.0379, radius=2.547, shapeName=dyn_hair_tail_5_capsule, center=<4.7408 0.1349 0>, rotation=<rot 90 0 0>}]
 dyn_hair_tail_MSTR [{shapeType=Capsule, length=1, radius=1, shapeName=dyn_hair_tail_MSTR_capsule}]
 dyn_tail_strand_low [{shapeType=Capsule, length=5.0954, radius=1.0984, shapeName=dyn_tail_strand_low_capsule, center=<5.1486 0 -1.8501>, rotation=<rot 90 0 0>}]
 dyn_tail_strand_up [{
   shapeType=Capsule
   length=9.6763
   radius=1.3185
   shapeName=dyn_tail_strand_up_capsule
   center=<9.4994 5.4974 -2.0336>
   rotation=<rot 51.2799 30.5624 17.7829>
  }]
 head [{shapeType=Capsule, length=3.1016, radius=8.2261, center=<2.842 2.8072 -0.4574>, rotation=<rot 90 0 0>}]
 neck_01 [{
   shapeType=Capsule
   length=8.3383
   radius=4.7808
   shapeName=neck_01_capsule
   center=<-0.0386 -0.8414 0.945>
   rotation=<rot 90 -179.4032 -176.7747>
  }]
 spine_04 [{shapeType=Capsule, length=19.5474, radius=10.4526, shapeName=spine_04_capsule, center=<0 0.1364 0>}]
 spine_05 [{shapeType=Capsule, length=10.0281, radius=9.2424, shapeName=spine_05_capsule, center=<0.3759 -1.9669 0>}]
 upperarm_l [{
   shapeType=Capsule
   length=17.0651
   radius=4.2245
   shapeName=upperarm_l_capsule
   center=<11.2808 -0.9453 -1.5579>
   rotation=<rot 87.4486 -179.9983 -179.9984>
  }]
}
constraints [
 {
  bone1Name=head
  bone2Name=neck_01
  swing1LimitDegrees=45
  swing1Motion=Free
  swing2LimitDegrees=45
  swing2Motion=Free
  twistLimitDegrees=45
  twistMotion=Free
 }
 {
  bone1Name=neck_01
  bone2Name=spine_05
  swing1LimitDegrees=45
  swing1Motion=Free
  swing2LimitDegrees=45
  swing2Motion=Free
  twistLimitDegrees=45
  twistMotion=Free
 }
 {
  bone1Name=upperarm_l
  bone2Name=clavicle_l
  swing1LimitDegrees=45
  swing1Motion=Free
  swing2LimitDegrees=45
  swing2Motion=Free
  twistLimitDegrees=45
  twistMotion=Free
 }
 {
  bone1Name=clavicle_l
  bone2Name=spine_05
  swing1LimitDegrees=45
  swing1Motion=Free
  swing2LimitDegrees=45
  swing2Motion=Free
  twistLimitDegrees=45
  twistMotion=Free
 }
 {bone1Name=dyn_hair_tail_1, bone2Name=dyn_hair_tail_MSTR, swing1LimitDegrees=75.84, swing2LimitDegrees=27.72, twistLimitDegrees=15}
 {bone1Name=dyn_hair_tail_2, bone2Name=dyn_hair_tail_1, swing1LimitDegrees=73.8, swing2LimitDegrees=54.24, twistLimitDegrees=35}
 {bone1Name=dyn_hair_tail_3, bone2Name=dyn_hair_tail_2, swing1LimitDegrees=102.44, swing2LimitDegrees=90, twistLimitDegrees=45}
 {bone1Name=dyn_hair_tail_4, bone2Name=dyn_hair_tail_3, swing1LimitDegrees=90, swing2LimitDegrees=90, twistLimitDegrees=20}
 {bone1Name=dyn_hair_tail_5, bone2Name=dyn_hair_tail_4, swing1LimitDegrees=70, swing2LimitDegrees=65, twistLimitDegrees=15}
 {
  bone1Name=clavicle_l
  bone2Name=dyn_hair_tail_2
  swing1LimitDegrees=45
  swing1Motion=Free
  swing2LimitDegrees=45
  swing2Motion=Free
  twistLimitDegrees=45
  twistMotion=Free
 }
 {
  bone1Name=spine_05
  bone2Name=spine_04
  swing1LimitDegrees=45
  swing1Motion=Free
  swing2LimitDegrees=45
  swing2Motion=Free
  twistLimitDegrees=45
  twistMotion=Free
 }
 {
  bone1Name=spine_05
  bone2Name=spine_04
  swing1LimitDegrees=45
  swing1Motion=Free
  swing2LimitDegrees=45
  swing2Motion=Free
  twistLimitDegrees=45
  twistMotion=Free
 }
 {bone1Name=dyn_tail_strand_low, bone2Name=dyn_hair_tail_2, swing1LimitDegrees=42.28, swing2LimitDegrees=40.84, twistLimitDegrees=10}
 {bone1Name=dyn_tail_strand_up, bone2Name=dyn_hair_tail_2, swing1LimitDegrees=20, swing2LimitDegrees=20.68, twistLimitDegrees=21.52}
 {bone1Name=dyn_hair_back, bone2Name=neck_01, swing1LimitDegrees=15, swing2LimitDegrees=10, twistLimitDegrees=5}
]
