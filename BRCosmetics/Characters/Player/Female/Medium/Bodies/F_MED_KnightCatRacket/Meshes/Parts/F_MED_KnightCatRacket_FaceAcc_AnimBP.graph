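# Animation Blueprint for the KnightCatRacket female character's face accessory part, adding secondary hair motion to poses copied from the attached parent and body meshes.
# In game, bRigidBodyEnabled picks RigidBody or AnimDynamics tail simulation, then fringe and right-side AnimDynamics chains follow.
# Bone offsets and constraints correct hair placement, driven partly by head-to-spine rotation, melee and parachute states.
# The front end instead uses a standalone RigidBody simulation with its own physics asset.
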AnimBlueprint /BRCosmetics/Characters/Player/Female/Medium/Bodies/F_MED_KnightCatRacket/Meshes/Parts/F_MED_KnightCatRacket_FaceAcc_AnimBP
target_skeleton=&/F_MED_KnightCatRacket_FaceAcc_Skeleton
graphs [
 AnimGraph {nodes [
   n1 = Copy Pose From Mesh(SourceMeshComponent=Get Skel_Body() -> (Skel_Body:Skeletal Mesh Component Object Reference))
   n2 = Copy Pose From Mesh() {bUseAttachedParent=true}
   n3 = RigidBody!(
    ComponentPose=Local To Component(LocalPose=Layered blend per bone(BasePose=@n2, BlendPoses_0=@n1) {layerSetup [{branchFilters [{boneName=clavicle_l}]}]})
    Alpha
    OverrideWorldGravity:Vector=Get Override World Gravity() -> (Override World Gravity:Vector)
    ExternalForce
   ) -> (Pose) {
    bClampLinearTranslationLimitToRefPose=true
    bOverrideWorldGravity=true
    overridePhysicsAsset=&/F_MED_KnightCatRacket_FaceAcc_Physics
    simulationSpace=BaseBoneSpace
    alphaBoolBlend {blendOption=Linear}
    baseBoneRef {boneName=spine_05}
    componentAppliedLinearAccClamp=<2000 2000 2000>
    componentLinearAccScale=<0.6 0.6 0.6>
    componentLinearVelScale=<0.6 0.6 0.6>
    overrideWorldGravity {z=-980}
    simSpaceSettings {maxAngularAcceleration=100}
   }
   n4 = Component To Local(ComponentPose=@n3)
   n5 = Copy Pose From Mesh() {bUseAttachedParent=true}
   n6 = Copy Pose From Mesh(SourceMeshComponent=Get Skel_Body())
   n7 = "Anim Dynamics - Chain: dyn_hair_tail_1 -> dyn_hair_tail_5"(
    ComponentPose=Local To Component(LocalPose=Layered blend per bone(BasePose=@n5, BlendPoses_0=@n6) {layerSetup [{branchFilters [{boneName=clavicle_l}]}]})
    Alpha=1
    ExternalForce="0,0,0"
   ) {
    angularDampingOverride=0.9
    angularSpringConstant=350
    bAngularSpring=true
    bChain=true
    bOverrideAngularDamping=true
    bUseGravityOverride=true
    bUsePlanarLimit=false
    bUseSphericalLimits=true
    simulationSpace=BoneRelative
    alphaBoolBlend {blendOption=Linear}
    boundBone {boneName=dyn_hair_tail_1}
    chainEnd {boneName=dyn_hair_tail_5}
    componentAppliedLinearAccClamp=<2000 2000 2000>
    componentLinearAccScale=<0.3 0.3 0.3>
    componentLinearVelScale=<0.3 0.3 0.3>
    gravityOverride {z=-980}
    physicsBodyDefinitions [
     {
      boundBone {boneName=dyn_hair_tail_1}
      boxExtents=<30 30 30>
      constraintSetup {angularLimitsMax=<0 15 10>, angularLimitsMin=<0 -15 -10>, angularTarget=<0 0 0>}
      localJointOffset=<-1 0 0>
     }
     {
      boundBone {boneName=dyn_hair_tail_2}
      boxExtents=<30 30 30>
      constraintSetup {angularLimitsMax=<0 10 15>, angularLimitsMin=<0 -10 -10>, angularTarget=<0 0 0>}
      localJointOffset=<-1 0 0>
     }
     {
      boundBone {boneName=dyn_hair_tail_3}
      boxExtents=<30 30 30>
      constraintSetup {angularLimitsMax=<0 10 20>, angularLimitsMin=<0 -10 -20>, angularTarget=<0 0 0>}
      localJointOffset=<-1 0 0>
     }
     {
      boundBone {boneName=dyn_hair_tail_4}
      boxExtents=<30 30 30>
      constraintSetup {angularLimitsMax=<0 10 25>, angularLimitsMin=<0 -10 -25>, angularTarget=<0 0 0>}
      localJointOffset=<-1 0 0>
     }
     {
      boundBone {boneName=dyn_hair_tail_5}
      boxExtents=<30 40 40>
      constraintSetup {angularLimitsMax=<0 10 25>, angularLimitsMin=<0 -10 -25>, angularTarget=<0 0 0>}
      localJointOffset=<-1 0 0>
     }
    ]
    sphericalLimits [
     {limitRadius=15, drivingBone {boneName=spine_05}, sphereLocalOffset=<-1.0834 -5.4109 -4.2378>}
     {limitRadius=12, drivingBone {boneName=upperarm_l}, sphereLocalOffset=<6.8146 -2.539 5.4177>}
     {limitRadius=13, drivingBone {boneName=spine_04}, sphereLocalOffset=<-0.933 -0.0605 -2.9864>}
     {limitRadius=7, drivingBone {boneName=neck_01}, sphereLocalOffset=<0 0 0>}
    ]
   }
   n8 = Anim Dynamics - Bone: dyn_tail_strand_low(ComponentPose=@n7, Alpha=1, ExternalForce="0,0,0") {
    angularDampingOverride=0.99
    angularSpringConstant=100
    bAngularSpring=true
    bOverrideAngularDamping=true
    bUseGravityOverride=true
    bUsePlanarLimit=false
    bUseSphericalLimits=true
    simulationSpace=BoneRelative
    alphaBoolBlend {blendOption=Linear}
    boundBone {boneName=dyn_tail_strand_low}
    componentAppliedLinearAccClamp=<2000 2000 2000>
    componentLinearAccScale=<1 1 1>
    componentLinearVelScale=<1 1 1>
    gravityOverride {z=-500}
    physicsBodyDefinitions [{
      boundBone {boneName=dyn_tail_strand_low}
      boxExtents=<5 5 5>
      constraintSetup {angularLimitsMax=<0 5 10>, angularLimitsMin=<0 -5 -15>, angularTarget=<0 0 0>}
      localJointOffset=<-1 0 0>
     }]
    relativeSpaceBone {boneName=spine_05}
    sphericalLimits [{limitRadius=15, drivingBone {boneName=spine_05}, sphereLocalOffset=<-6.4392 0 0>}]
   }
   n9 = Anim Dynamics - Bone: dyn_tail_strand_up(ComponentPose=@n8, Alpha=1, ExternalForce="0,0,0") {
    angularDampingOverride=0.99
    angularSpringConstant=100
    bAngularSpring=true
    bOverrideAngularDamping=true
    bUseGravityOverride=true
    bUsePlanarLimit=false
    bUseSphericalLimits=true
    simulationSpace=BoneRelative
    alphaBoolBlend {blendOption=Linear}
    boundBone {boneName=dyn_tail_strand_up}
    componentAppliedLinearAccClamp=<2000 2000 2000>
    componentLinearAccScale=<1 1 1>
    componentLinearVelScale=<1 1 1>
    gravityOverride {z=-500}
    physicsBodyDefinitions [{
      boundBone {boneName=dyn_tail_strand_up}
      boxExtents=<5 5 5>
      constraintSetup {angularLimitsMax=<0 5 10>, angularLimitsMin=<0 -5 -15>, angularTarget=<0 0 0>}
      localJointOffset=<-1 0 0>
     }]
    relativeSpaceBone {boneName=spine_05}
    sphericalLimits [{limitRadius=15, drivingBone {boneName=spine_05}, sphereLocalOffset=<-5.9752 0 0>}]
   }
   n10 = Component To Local(ComponentPose=@n9)
   n11 = Get bRigidBodyEnabled()
   n12 = "Anim Dynamics - Chain: dyn_hair_fringe_1 -> dyn_hair_fringe_3"(
    ComponentPose=Local To Component(LocalPose=Blend Poses by bool(BlendPose_0=@n4, BlendPose_1=@n10, bActiveValue=@n11) {blendType=Linear})
    Alpha=1
    ExternalForce="0,0,0"
   ) {
    angularDampingOverride=0.9
    angularSpringConstant=170
    bAngularSpring=true
    bChain=true
    bOverrideAngularDamping=true
    bUseGravityOverride=true
    bUsePlanarLimit=false
    bUseSphericalLimits=true
    simulationSpace=BoneRelative
    alphaBoolBlend {blendOption=Linear}
    boundBone {boneName=dyn_hair_fringe_1}
    chainEnd {boneName=dyn_hair_fringe_3}
    componentAppliedLinearAccClamp=<2000 2000 2000>
    componentLinearAccScale=<0.1 0.1 0.1>
    componentLinearVelScale=<0.1 0.1 0.1>
    gravityOverride {z=-500}
    physicsBodyDefinitions [
     {
      boundBone {boneName=dyn_hair_fringe_1}
      boxExtents=<40 60 40>
      constraintSetup {angularLimitsMax=<0 0 10>, angularLimitsMin=<0 -8 -10>, angularTarget=<0 0 0>}
      localJointOffset=<0 0 0>
     }
     {
      boundBone {boneName=dyn_hair_fringe_2}
      boxExtents=<45 60 45>
      constraintSetup {angularLimitsMax=<0 0 5>, angularLimitsMin=<0 -4 -5>, angularTarget=<0 0 0>}
      localJointOffset=<0 0 0>
     }
     {
      boundBone {boneName=dyn_hair_fringe_3}
      boxExtents=<30 30 30>
      constraintSetup {angularLimitsMax=<5 5 0>, angularLimitsMin=<-5 -5 -10>, angularTarget=<0 0 0>}
      localJointOffset=<0 0 0>
     }
    ]
    relativeSpaceBone {boneName=neck_02}
    sphericalLimits [{limitRadius=10.5, drivingBone {boneName=head}, sphereLocalOffset=<1.9972 -0.7239 -1.328>}]
   }
   n13 = "Anim Dynamics - Chain: dyn_hair_right_1 -> dyn_hair_right_2"(ComponentPose=@n12, Alpha=1, ExternalForce="0,0,0") {
    angularDampingOverride=0.99
    angularSpringConstant=150
    bAngularSpring=true
    bChain=true
    bOverrideAngularDamping=true
    bUseGravityOverride=true
    bUsePlanarLimit=false
    bUseSphericalLimits=true
    simulationSpace=BoneRelative
    alphaBoolBlend {blendOption=Linear}
    boundBone {boneName=dyn_hair_right_1}
    chainEnd {boneName=dyn_hair_right_2}
    componentAppliedLinearAccClamp=<2000 2000 2000>
    componentLinearAccScale=<0.3 0.3 0.3>
    componentLinearVelScale=<0.3 0.3 0.3>
    physicsBodyDefinitions [
     {
      boundBone {boneName=dyn_hair_right_1}
      boxExtents=<20 20 20>
      constraintSetup {angularLimitsMax=<0 5 10>, angularLimitsMin=<0 -5 -25>, angularTarget=<0 0 0>}
      localJointOffset=<-0.5 0 0>
     }
     {
      boundBone {boneName=dyn_hair_right_2}
      boxExtents=<20 20 20>
      constraintSetup {angularLimitsMax=<0 5 10>, angularLimitsMin=<0 -5 -25>, angularTarget=<0 0 0>}
      localJointOffset=<-0.5 0 0>
     }
    ]
    relativeSpaceBone {boneName=spine_05}
    sphericalLimits [
     {limitRadius=20, drivingBone {boneName=spine_05}, sphereLocalOffset=<-1.8476 0.9953 -2.8729>}
     {limitRadius=6, drivingBone {boneName=neck_01}, sphereLocalOffset=<0 0 5.3403>}
    ]
   }
   n14 = "Transform (Modify) Bone - Bone: dyn_hair_right_2"(ComponentPose=@n13, Alpha=Get Hair_fix() -> (Hair_fix:double), Translation="0,0,0", Rotation="0.000,-103.804,0.000", Scale="1,1,1") {
    rotationMode=BMM_Additive
    rotationSpace=BCS_BoneSpace
    translationMode=BMM_Additive
    translationSpace=BCS_BoneSpace
    alphaBoolBlend {blendOption=Linear}
    boneToModify {boneName=dyn_hair_right_2}
    rotation=<rot 0 -103.804 0>
   }
   n15 = "Transform (Modify) Bone - Bone: dyn_hair_right_2"(ComponentPose=@n14, Alpha=Get Hair_fix_onMelee() -> (Hair_fix_onMelee:double), Translation="0,-5,0", Rotation="0,0,0", Scale="1,1,1") {
    rotationMode=BMM_Additive
    rotationSpace=BCS_BoneSpace
    translationMode=BMM_Additive
    translationSpace=BCS_BoneSpace
    alphaBoolBlend {blendOption=Linear}
    boneToModify {boneName=dyn_hair_right_2}
    translation {y=-5}
   }
   n16 = "Transform (Modify) Bone - Bone: dyn_hair_right_2"!(ComponentPose=@n15, bAlphaBoolEnabled:bool=True, Translation="0,0,0", Rotation="0,-60,0", Scale="1,1,1") -> (Pose) {
    alphaInputType=Bool
    rotationMode=BMM_Additive
    rotationSpace=BCS_BoneSpace
    translationMode=BMM_Additive
    translationSpace=BCS_BoneSpace
    alphaBoolBlend {blendInTime=0.3, blendOption=CubicInOut, blendOutTime=0.3}
    boneToModify {boneName=dyn_hair_right_2}
    rotation {yaw=-60}
   }
   n17 = "Transform (Modify) Bone - Bone: dyn_hair_tail_MSTR"!(ComponentPose=@n16, bAlphaBoolEnabled:bool=True, Translation="-5,3,0", Rotation="30,0,0", Scale="1,1,1") -> (Pose) {
    alphaInputType=Bool
    rotationMode=BMM_Additive
    rotationSpace=BCS_BoneSpace
    translationMode=BMM_Additive
    translationSpace=BCS_BoneSpace
    alphaBoolBlend {blendOption=Linear}
    boneToModify {boneName=dyn_hair_tail_MSTR}
    rotation {pitch=30}
    translation=<-5 3>
   }
   n18 = "Transform (Modify) Bone - Bone: dyn_hair_back"!(ComponentPose=@n17, bAlphaBoolEnabled:bool=True, Translation="-2,0,0", Rotation="0,0,0", Scale="1,1,1") -> (Pose) {
    alphaInputType=Bool
    rotationMode=BMM_Additive
    rotationSpace=BCS_BoneSpace
    translationMode=BMM_Additive
    translationSpace=BCS_BoneSpace
    alphaBoolBlend {blendInTime=0.5, blendOption=CubicInOut, blendOutTime=0.5}
    boneToModify {boneName=dyn_hair_back}
    translation {x=-2}
   }
   n19 = Get bIsParachuteOpened()
   n20 = "Constraint - dyn_hair_back (P)"!(ComponentPose=@n18, bAlphaBoolEnabled:bool=@n19, ConstraintWeights_0=1) -> (Pose) {
    alphaInputType=Bool
    alphaBoolBlend {blendInTime=0.2, blendOption=Linear, blendOutTime=0.2}
    boneToModify {boneName=dyn_hair_back}
    constraintSetup [{targetBone {boneName=neck_02}}]
    constraintWeights [1]
   }
   n21 = "Transform (Modify) Bone - Bone: dyn_hair_back"(ComponentPose=@n20, bAlphaBoolEnabled=@n19, Translation="-2,1,-0.5", Rotation="0,0,0", Scale="1,1,1") {
    alphaInputType=Bool
    rotationMode=BMM_Additive
    rotationSpace=BCS_BoneSpace
    translationMode=BMM_Additive
    translationSpace=BCS_BoneSpace
    alphaBoolBlend {blendInTime=0.2, blendOption=Linear, blendOutTime=0.2}
    boneToModify {boneName=dyn_hair_back}
    translation=<-2 1 -0.5>
   }
   n22 = RigidBody(ComponentPose=Local To Component(LocalPose=Copy Pose From Mesh() {bUseAttachedParent=true}), OverrideWorldGravity="0,0,-980") {
    bClampLinearTranslationLimitToRefPose=true
    bOverrideWorldGravity=true
    overridePhysicsAsset=&/F_MED_KnightCatRacket_FaceAcc_Physics_Frontend
    simulationSpace=BaseBoneSpace
    alphaBoolBlend {blendOption=Linear}
    baseBoneRef {boneName=spine_05}
    componentAppliedLinearAccClamp=<2000 2000 2000>
    componentLinearAccScale=<0.6 0.6 0.6>
    componentLinearVelScale=<0.6 0.6 0.6>
    overrideWorldGravity {z=-980}
    simSpaceSettings {maxAngularAcceleration=100}
   }
   Output Pose(Result=Blend Poses by bool(BlendPose_0=Component To Local(ComponentPose=@n22), BlendPose_1=Component To Local(ComponentPose=@n21), bActiveValue=Get bIsInFrontEnd()) {blendType=Linear}) {layerGroup=None}
  ]}
 EventGraph {nodes [
   Event BlueprintUpdateAnimation()
   > n1 = Gravity Override State Function DEPRECATED(CharacterPartAnimInstance=Self-Reference(), GameTimeInSeconds=GetGameTimeInSeconds(), GravityOverrideParams=Get Gravity Override Params())
   > Set Override World Gravity(Override World Gravity:Vector=@n1) -> (Output_Get:Vector)
   n2 = GetAlphaFromReferenceTransformDelta(SkeletalMeshComponent=Get Skel_Body(), BoneName=head, BaseName=spine_05, InputRange=(X=-30,Y=-60), TwistAxis="1,0,0")
   > Set Hair_fix(Hair_fix:double=@n2) -> (Output_Get:double)
   > Set Jump_Fall(Jump_Fall:bool=OR Boolean(A=Get bIsFalling(), B=Get bIsJumping())) -> (Output_Get:bool)
   n3 = Get bIsPlayingMeleeAnim()
   > Set Hair_fix_onMelee(Hair_fix_onMelee:double=float * float(A=@n2, B=SelectFloat(A=1, B=0, bPickA=@n3)) {is=title K2Node_PromotableOperator}) -> (Output_Get:double)
   Event BlueprintInitializeAnimation()
   > n4 = Cast To PlayerPawn_Athena(Object=TryGetPawnOwner())
   > Set Skel_Body(Skel_Body:Skeletal Mesh Component Object Reference=GetSkeletalMeshForPartType(self=@n4)) -> (Output_Get:Skeletal Mesh Component Object Reference)
   PrintString!(execute, InString="To String (Boolean)"(InBool=Get bIsInFrontEnd()), bPrintToScreen, bPrintToLog, TextColor, Duration, Key) -> (then)
  ]}
]
variables [
 Character Part Anim Instance {type=CustomCharacterPartAnimInstance, default=null}
 Gravity Override Params {type=GravityOverrideParamsStruct}
 Hair_fix {type=number}
 Hair_fix_onMelee {type=number}
 Jump_Fall {type=boolean}
 Override World Gravity {type=Vector}
 Skel_Body {type=SkeletalMeshComponent}
]
